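# MV_EmptyCloudMat is a placeholder volume-domain material for volumetric clouds in the TimeOfDay clouds and storm content, with no cloud shaping of its own.
# It uses additive blending and is flagged for use with Volumetric Cloud.
# The graph holds a single constant with no connection shown to any material output, so it has no noise, textures or parameters.
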Material /Game/TimeOfDay/CloudsAndStorm/Materials/MV_EmptyCloudMat
graph {
 properties {
  bUsedWithVolumetricCloud=true
  blendMode=BLEND_Additive
  materialDomain=MD_Volume
  pixelDepthOffsetMode=PDOM_Legacy
  usedShadingModels=Default Lit
 }
 nodes [Constant() {r=0.5}]
}
