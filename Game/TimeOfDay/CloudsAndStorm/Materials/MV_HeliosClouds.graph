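# Volumetric cloud material that renders the sky's cloud layer and the storm around the safe zone.
# An SDF volume texture and animated noise shape the clouds, with time-of-day colors and scattering occlusion.
# MPC_SafeZoneClouds drives the storm, built from a profile texture, Voronoi noise, rotating 3D flowmaps and radial waves, plus lightning.
# Conservative density and empty-space skipping limit raymarch cost, and reflection captures omit the storm.
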Material /Game/TimeOfDay/CloudsAndStorm/Materials/MV_HeliosClouds
graph {
 properties {
  bUsedWithVolumetricCloud=true
  blendMode=BLEND_Additive
  materialDomain=MD_Volume
  pixelDepthOffsetMode=PDOM_Legacy
  usedShadingModels=Default Lit
 }
 nodes [
  n1 = Subtract(A=NamedRerouteUsage(), B=NamedRerouteUsage())
  Abs(None=@n1)
  n2 = Multiply(A=NamedRerouteUsage(), B=ScalarParameter() {defaultValue=26, parameterName=offsetNoiseFrequencyb})
  n3 = Multiply(A=NamedRerouteUsage(), B=ScalarParameter() {defaultValue=0.01, desc=0.002, parameterName=NoiseSpeedMult})
  n4 = ConstantBiasScale(None=TextureSample!(UVs=Add(A=@n2, B=@n3), Tex, Level, Apply View MipBias) {mipValueMode=TMVM_MipLevel, samplerType=SAMPLERTYPE_LinearColor, texture=/Game/Global/MaterialFunctions/VT_ClothSimWPO}.R) {bias=-0.5, scale=2}
  n5 = Multiply(A=@n4, B=ScalarParameter() {defaultValue=0.1, desc=0.1, parameterName=NoiseGain})
  n6 = ScalarParameter() {defaultValue=0.3, parameterName=CloudFlow}
  n7 = NamedRerouteUsage()
  Add(A=@n5, B=Multiply(A=NamedRerouteUsage(), B=LinearInterpolate(A=@n6, Alpha=@n7) {constB=0}))
  CollectionParameter() {collection=/Game/Athena/SafeZone/MPC_SafeZoneClouds, parameterName=CloudStartingBias}
  CollectionParameter() {collection=&MPC_SafeZoneClouds, parameterName=SafeZoneScale}
  CollectionParameter() {collection=&MPC_SafeZoneClouds, parameterName=SafeZoneScale}
  CollectionParameter() {collection=&MPC_SafeZoneClouds, parameterName=CloudStartingBias}
  ComponentMask(None=NamedRerouteUsage()) {g=true}
  n8 = VolumetricAdvancedMaterialInput()
  ComponentMask(None=@n8.ConservativeDensity as Float4) {r=true}
  Constant()
  n9 = CameraPositionWS()
  n10 = ScalarParameter() {parameterName=AltitudeBlendingLow}
  n11 = Divide(A=Subtract(A=ComponentMask(None=@n9) {b=true}, B=@n10), B=Subtract(A=ScalarParameter() {parameterName=AltitudeBlendingHigh}, B=@n10))
  LinearInterpolate(
   A=ScalarParameter() {defaultValue=0.2, group=Scattering and Phase, parameterName=MS Occ}
   B=ScalarParameter() {defaultValue=0.2, group=Scattering and Phase, parameterName=MS Occ HighAltitude}
   Alpha=NamedRerouteDeclaration(None=Saturate(None=@n11)) {name=Altitude Blend, nodeColor=<rgba 1 0 0.1641 1>}
  )
  n12 = NamedRerouteUsage()
  MaterialFunctionCall(
   Input=@n12
   Input Low=ScalarParameter() {defaultValue=0.5, parameterName=max}
   Input High=ScalarParameter() {defaultValue=-0.1667, parameterName=min}
  ) {materialFunction=&RemapValue_0-1}
  n13 = ScalarParameter() {defaultValue=-0.2, parameterName=NoiseLow}
  n14 = ScalarParameter() {defaultValue=0.5, parameterName=NoiseHigh}
  MaterialFunctionCall(
   Input=NamedRerouteDeclaration(None=CloudSampleAttribute().NormAltitudeInLayer) {name=NormalizedAltitude, nodeColor=<rgba 0.5 1 0 1>}
   Input Low=@n13
   Input High=@n14
  ) {materialFunction=&RemapValue_0-1}
  n15 = ScalarParameter() {defaultValue=2250000, parameterName=worldScale}
  MaterialFunctionCall!(Cm=@n15) {materialFunction=/Engine/Functions/Engine_MaterialFunctions01/Units/Cm-to-Km}
  n16 = Constant() {r=1}
  MaterialFunctionCall(X=@n16, Y=@n16, Z=Multiply(A=Time(), B=ScalarParameter() {defaultValue=0.001, parameterName=VerticalNioseSpeed}) {constB=0.05}) {materialFunction=&MakeFloat3}
  n17 = NamedRerouteUsage()
  n18 = Saturate(None=MaterialFunctionCall(Input=CloudSampleAttribute().NormAltitudeInLayer, Input Low=NamedRerouteUsage(), Input High=NamedRerouteUsage()) {materialFunction=&RemapValueRange}.Result)
  n19 = TextureSample!(
   UVs=AppendVector(A=ComponentMask(None=@n17) {g=true, r=true}, B=NamedRerouteDeclaration(None=@n18) {name=CloudUVZ, nodeColor=<rgba 1 0 0.75 1>})
   Tex
   Apply View MipBias
  ) {
   samplerSource=SSM_Wrap_WorldGroupSettings
   samplerType=SAMPLERTYPE_LinearColor
   texture=/Game/TimeOfDay/CloudsAndStorm/Textures/VolumeCloudSDF_v04_Volume
  }
  MaterialFunctionCall!(value=@n19.R, gain=ScalarParameter() {defaultValue=0.65, parameterName=Gain}) {materialFunction=/Game/Global/MaterialFunctions/MF_Gain}
  n20 = Time()
  n21 = CollectionParameter() {collection=&MPC_SafeZoneClouds, parameterName=CustomCloudsTime}
  n22 = Ceil(None=Saturate(None=@n21))
  Multiply(
   A=NamedRerouteDeclaration(None=LinearInterpolate(A=@n20, B=@n21, Alpha=@n22)) {name=CustomTime, nodeColor=<rgba 0 1 0.1094 1>}
   B=CollectionParameter() {collection=&MPC_SafeZoneClouds, parameterName=TopCloudsSpeed}
  )
  Multiply(A=NamedRerouteUsage(), B=@n4)
  Multiply(A=NamedRerouteUsage(), B=NamedRerouteUsage()) {constB=0}
  n23 = Saturate(None=Add(A=NamedRerouteUsage(), B=ScalarParameter() {defaultValue=0.05, parameterName=StormMaskBias}))
  n24 = Add(A=Power(Base=OneMinus(None=@n23), Exp=ScalarParameter() {defaultValue=7, parameterName=Exp_2}), B=NamedRerouteUsage())
  NamedRerouteDeclaration(None=LinearInterpolate(B=Saturate(None=@n24), Alpha=ScalarParameter() {defaultValue=1, parameterName=UseCloudMask}) {constA=1}) {name=CloudMask, nodeColor=<rgba 0.8281 1 0 1>}
  n25 = NamedRerouteUsage()
  n26 = VectorParameter() {group=Storm, parameterName=AlbedoStormTop, defaultValue=<rgba 0 0.0938 0.0649 1>}
  n27 = NamedRerouteUsage()
  n28 = LinearInterpolate(A=@n25, B=@n26.RGB, Alpha=Multiply(A=ComponentMask(None=@n27) {b=true}, B=NamedRerouteUsage()))
  n29 = VectorParameter() {group=Storm, parameterName=AlbedoStormBottom, defaultValue=<rgba 0.2602 0.1591 0.2656 1>}
  n30 = NamedRerouteUsage()
  n31 = LinearInterpolate(
   A=@n28
   B=LinearInterpolate(A=LinearInterpolate(A=@n29.RGB, B=@n26.RGB, Alpha=@n30), B=@n25, Alpha=NamedRerouteUsage())
   Alpha=NamedRerouteUsage()
  )
  NamedRerouteDeclaration(None=StaticSwitchParameter(True=ReflectionCapturePassSwitch!(Default=@n31, Reflection=@n28), False=@n28) {defaultValue=true, group=Storm, parameterName=EnableStorm}) {name=CloudsAlbedo_1, nodeColor=<rgba 0 0.625 1 1>}
  n32 = NamedRerouteDeclaration(None=MaterialFunctionCall!(WorldScale=NamedRerouteUsage()) {materialFunction=/Game/TimeOfDay/CloudsAndStorm/Materials/MaterialFunctions/MF_WorldRemap}.Result) {name=UVBOX, nodeColor=<rgba 0 0.7656 1 1>}
  n33 = Multiply(A=ComponentMask(None=NamedRerouteUsage()) {g=true, r=true}, B=ScalarParameter() {defaultValue=0.5, parameterName=SphereMaskErosionGain}) {constB=0.2}
  n34 = SphereMask(
   A=ReflectionCapturePassSwitch(Default=Subtract(A=@n32, B=@n33), Reflection=@n32)
   B=Constant()
   Radius=ScalarParameter() {defaultValue=0.2, parameterName=maskRadius}
   Hardness=ScalarParameter() {defaultValue=0.333, parameterName=MaskHardness}
  )
  NamedRerouteDeclaration(None=Saturate(None=OneMinus(None=@n34))) {name=voidMask, nodeColor=<rgba 0 0.0391 1 1>}
  n35 = MaterialFunctionCall(
   Color_Morning=VectorParameter() {group=Clouds, parameterName=CloudsEmissiveMorning, defaultValue=<rgba 0 0 0 1>}.RGB
   Color_Day=VectorParameter() {group=Clouds, parameterName=CloudsEmissiveDay, defaultValue=<rgba 0.005 0.005 0.005 1>}.RGB
   Color_Evening=VectorParameter() {group=Clouds, parameterName=CloudsEmissiveEvening, defaultValue=<rgba 0.002 0.0026 0.009 1>}.RGB
   Color_Night=VectorParameter() {group=Clouds, parameterName=CloudsEmissiveNight, defaultValue=<rgba 0.0009 0.0067 0.007 1>}.RGB
  ) {materialFunction=&Color_Time_of_Day}
  n36 = OneMinus(None=NamedRerouteUsage())
  n37 = Add(A=Multiply(A=NamedRerouteUsage(), B=NamedRerouteUsage()), B=Multiply(A=NamedRerouteUsage(), B=Multiply(A=@n35.Result, B=@n36)))
  n38 = Constant()
  NamedRerouteDeclaration(None=StaticSwitchParameter(True=ReflectionCapturePassSwitch(Default=@n37, Reflection=@n38), False=@n38) {defaultValue=true, group=Storm, parameterName=EnableStorm}) {name=CloudsEmissive_1, nodeColor=<rgba 0.8516 1 0 1>}
  n39 = NamedRerouteUsage()
  n40 = VectorParameter() {group=Clouds, parameterName=ecx, defaultValue=<rgba 0.5 0.6 0.7 0.5>}
  n41 = Multiply(A=NamedRerouteUsage(), B=Multiply(A=@n39, B=Multiply(A=@n40.RGB, B=@n40.RGB)))
  n42 = VectorParameter() {group=Storm, parameterName=ecx_Storm, defaultValue=<rgba 0.5 0.5 0.7 0.5>}
  n43 = Max(A=@n41, B=Multiply(A=NamedRerouteUsage(), B=Multiply(A=@n42.RGB, B=@n42.RGB)))
  NamedRerouteDeclaration(None=StaticSwitchParameter(True=ReflectionCapturePassSwitch(Default=@n43, Reflection=@n41), False=@n41) {defaultValue=true, group=Storm, parameterName=EnableStorm}) {name=Final Extinction, nodeColor=<rgba 0.4702 1 0.0199 1>}
  n44 = ComponentMask(None=@n8.ConservativeDensity as Float4) {g=true}
  n45 = ScalarParameter() {defaultValue=0.5, parameterName=max}
  n46 = ScalarParameter() {defaultValue=-0.1667, parameterName=min}
  n47 = MaterialFunctionCall(Input=Subtract(A=@n44, B=NamedRerouteUsage()), Input Low=@n45, Input High=@n46) {materialFunction=&RemapValue_0-1}
  n48 = NamedRerouteUsage()
  n49 = MaterialFunctionCall(Input=@n44, Input Low=@n45, Input High=@n46) {materialFunction=&RemapValue_0-1}
  n50 = ScalarParameter() {defaultValue=0.064, parameterName=reflectionDensity}
  NamedRerouteDeclaration(None=ReflectionCapturePassSwitch(Default=Multiply(A=@n47.Result, B=@n48) {constB=0.0001}, Reflection=Multiply(A=@n49.Result, B=@n50) {constB=0.0001})) {name=puffyCloudDensity, nodeColor=<rgba 0 1 0.4844 1>}
  n51 = CloudSampleAttribute()
  n52 = ScalarParameter() {parameterName=LowAltRemapCol}
  n53 = ScalarParameter() {defaultValue=1, parameterName=HIghAltRemapCol}
  n54 = OneMinus(None=Saturate(None=MaterialFunctionCall(Input=@n51.NormAltitudeInLayer, Input Low=@n52, Input High=@n53) {materialFunction=&RemapValueRange}.Result))
  NamedRerouteDeclaration(None=OneMinus(None=Power(Base=@n54))) {name=rampColour, nodeColor=<rgba 0.6875 1 0 1>}
  n55 = NamedRerouteUsage()
  NamedRerouteDeclaration(None=Multiply(A=@n5, B=LinearInterpolate(A=@n13, B=@n14, Alpha=@n55))) {name=CloudNoise, nodeColor=<rgba 0 0.6953 1 1>}
  n56 = Time()
  NamedRerouteDeclaration(None=MaterialFunctionCall(X=@n56, Y=@n56, Z=Constant()) {materialFunction=&MakeFloat3}.Result) {name=opposingWind, nodeColor=<rgba 0.6172 0 1 1>}
  NamedRerouteDeclaration(None=MaterialFunctionCall!(value=@n19.R, bias=ScalarParameter() {defaultValue=0.6, parameterName=SDFBias}) {materialFunction=/Game/Global/MaterialFunctions/MF_Bias}.Result) {name=RawCloudSDF, nodeColor=<rgba 1 0.5391 0 1>}
  n57 = ScalarParameter() {defaultValue=0.6, parameterName=CloudConservativeExpand}
  n58 = Sign(None=Multiply(A=Subtract(A=@n12, B=@n57)) {constB=-1})
  NamedRerouteDeclaration(None=Multiply(A=Saturate(None=@n58), B=NamedRerouteUsage()) {constB=0}) {name=ConservativeClouds, nodeColor=<rgba 0 0.0156 1 1>}
  n59 = NamedRerouteUsage()
  n60 = Multiply(A=Add(A=WorldPosition().XYZ, B=Multiply(A=@n59) {constB=0.5}), B=Divide(A=Constant() {r=1}, B=@n59))
  n61 = Multiply(A=NamedRerouteUsage(), B=NamedRerouteUsage())
  n62 = Multiply(A=ComponentMask(None=@n61) {r=true}) {constB=0.6667}
  NamedRerouteDeclaration(None=Subtract(
    A=@n60
    B=MaterialFunctionCall(X=@n61, Y=@n61, Z=@n62) {materialFunction=&MakeFloat3}.Result
   )) {name=worldUV, nodeColor=<rgba 0 0.7188 1 1>}
  NamedRerouteDeclaration(None=@n15) {name=Scale, nodeColor=<rgba 1 0 0.8203 1>}
  n63 = Distance(
   A=ComponentMask(None=WorldPosition().XYZ) {g=true, r=true}
   B=ComponentMask(None=CollectionParameter() {collection=&MPC_SafeZoneClouds, parameterName=SafeZoneLoc}) {g=true, r=true}
  )
  n64 = MaterialFunctionCall(Float2=NamedRerouteUsage()) {materialFunction=&BreakOutFloat2Components}
  n65 = MaterialFunctionCall!(Km=@n64.G) {materialFunction=/Engine/Functions/Engine_MaterialFunctions01/Units/Km-to-Cm}
  n66 = ScalarParameter() {defaultValue=1, group=StormRadialWaves, parameterName=WavesInnerRadiusMultiplier}
  n67 = Subtract(A=@n63, B=@n65.Cm)
  n68 = MaterialFunctionCall(Km=@n64.R) {materialFunction=&Km-to-Cm}
  n69 = Add(A=@n68.Cm, B=@n65.Cm)
  n70 = NamedRerouteUsage()
  n71 = Divide(A=Subtract(A=CloudSampleAttribute().NormAltitudeInLayer, B=ComponentMask(None=@n70) {b=true}), B=ComponentMask(None=@n70) {a=true})
  n72 = Add(A=Saturate(None=Divide(A=@n67, B=@n69)), B=Saturate(None=OneMinus(None=@n71)))
  n73 = Add(
   A=MaterialFunctionCall!(Vector or UVs=NamedRerouteUsage()) {materialFunction=/Game/Athena/SafeZone/S11/MF_VectorToRadialValueFast}.Result
   B=Multiply(A=Time(), B=ScalarParameter() {group=StormRadialWaves, parameterName=WavesSpeed})
  )
  n74 = Sine(None=Multiply(A=Frac(None=@n73), B=ScalarParameter() {defaultValue=4, group=StormRadialWaves, parameterName=WaveCount}) {constB=10})
  n75 = Power(
   Base=Add(A=Multiply(A=@n74) {constB=0.5}) {constB=0.5}
   Exp=ScalarParameter() {defaultValue=1, group=StormRadialWaves, parameterName=WavesPow}
  )
  n76 = ComponentMask(None=Saturate(None=Add(A=@n72, B=@n75) {constB=0})) {r=true}
  n77 = OneMinus(None=@n76)
  n78 = Multiply(A=@n65.Cm, B=Saturate(None=Add(A=@n66, B=@n77)))
  NamedRerouteDeclaration(None=Divide(A=Subtract(A=@n63, B=@n78), B=@n68.Cm)) {name=Storm Radius_1, nodeColor=<rgba 0.0547 1 0 1>}
  n79 = ScalarParameter() {defaultValue=128, group=Storm, parameterName=Global Storm Density}
  n80 = ScalarParameter() {defaultValue=128, group=Storm, parameterName=Global Storm Density Top}
  n81 = NamedRerouteUsage()
  n82 = ScalarParameter() {defaultValue=0.001, group=Storm, parameterName=scc}
  n83 = Saturate(None=Add(A=Multiply(A=@n81, B=@n82), B=ScalarParameter() {defaultValue=0.5, group=Storm, parameterName=off}))
  n84 = ComponentMask(None=NamedRerouteUsage()) {r=true}
  n85 = LinearInterpolate(
   A=ScalarParameter() {defaultValue=0.5286, group=StormDFs, parameterName=Noise Intensity}
   B=ScalarParameter() {defaultValue=0.2, group=StormDFs, parameterName=Noise IntensityTop}
   Alpha=@n83
  )
  n86 = Multiply(A=NamedRerouteUsage(), B=VectorParameter() {group=Storm, parameterName=WorldScaleNoise, defaultValue=<rgba 1 1 1 0>}.RGB)
  n87 = LinearInterpolate(
   A=ScalarParameter() {defaultValue=1, group=Storm, parameterName=NoiseGlobalScaleEnd}
   B=ScalarParameter() {defaultValue=1, group=Storm, parameterName=NoiseGlobalScale}
   Alpha=NamedRerouteUsage()
  )
  n88 = Multiply(A=ScalarParameter() {defaultValue=4500000, group=Storm, parameterName=scalenew}, B=@n87)
  n89 = TextureSample!(
   UVs=Divide(A=@n86, B=MaterialFunctionCall(Km=@n88) {materialFunction=&Km-to-Cm}.Cm)
   Tex
   Level=CollectionParameter() {collection=&MPC_SafeZoneClouds, parameterName=StormDetailVolumeMip}
   Apply View MipBias
  ) {mipValueMode=TMVM_MipLevel, samplerType=SAMPLERTYPE_LinearColor, texture=/Volumetrics/Content/VolumeTextures/Textures/VT_Voronoi_Simple}
  n90 = ScalarParameter() {defaultValue=0.8, group=StormDFs, parameterName=Noise DF Center End}
  n91 = ScalarParameter() {defaultValue=0.8, group=StormDFs, parameterName="Noise DF Center "}
  n92 = NamedRerouteUsage()
  n93 = Multiply(
   A=NamedRerouteDeclaration(None=Multiply(A=@n84, B=@n85)) {name=StormDebug, nodeColor=<rgba 0.3828 1 0 1>}
   B=Subtract(A=@n89.R, B=LinearInterpolate(A=@n90, B=@n91, Alpha=@n92)) {constB=0.5}
  )
  n94 = ScalarParameter() {defaultValue=0.1, group=StormFlowMap, parameterName=T}
  n95 = NamedRerouteUsage()
  n96 = MaterialFunctionCall(Float2=Normalize(VectorInput=NamedRerouteUsage())) {materialFunction=&BreakOutFloat2Components}
  n97 = AppendVector(A=AppendVector(A=Multiply(A=@n96.G) {constB=-1}, B=@n96.R), B=Constant())
  n98 = ScalarParameter() {defaultValue=0.7, group=StormFlowMap, parameterName=Storm Rotation Velocity}
  n99 = MaterialFunctionCall(Km=Multiply(A=@n87, B=ScalarParameter() {defaultValue=512, group=Storm, parameterName=scale2})) {materialFunction=&Km-to-Cm}
  n100 = ComponentMask(None=MaterialFunctionCall!(
    Volume Texture=TextureObject() {samplerType=SAMPLERTYPE_LinearColor, texture=&VT_Voronoi_Simple}
    UVW=Divide(A=@n86, B=@n99.Cm)
    Velocity=Multiply(A=@n97, B=@n98)
    Override Time=Multiply(A=@n94, B=@n95)
    Mip Level=CollectionParameter() {collection=&MPC_SafeZoneClouds, parameterName=StormFlowMapsVolumeMip}
   ) {materialFunction=/Engine/Functions/Engine_MaterialFunctions02/Texturing/Flowmaps_3D}.Result RGB) {r=true}
  n101 = ScalarParameter() {defaultValue=0.8, group=StormDFs, parameterName="Flowmaps DF Center "}
  n102 = ScalarParameter() {defaultValue=1, group=StormDFs, parameterName="Flowmaps DF Center PreStart "}
  n103 = Saturate(None=NamedRerouteUsage())
  n104 = Multiply(
   A=Subtract(A=NamedRerouteDeclaration(None=@n100) {name=FlowMap_1, nodeColor=<rgba 1 0 0.1172 1>}, B=LinearInterpolate(A=@n101, B=@n102, Alpha=@n103)) {constB=0.5}
   B=ScalarParameter() {defaultValue=0.5, group=StormDFs, parameterName=Flowmaps Intensity}
  )
  n105 = ScalarParameter() {defaultValue=0.5286, group=StormDFs, parameterName=StormDonut Intensity}
  MaterialExpressionNamedRerouteUsage_99 = NamedRerouteUsage()
  MaterialExpressionComponentMask_34 = ComponentMask(None=NamedRerouteUsage()) {g=true}
  n106 = Custom!(NormAltitude=@MaterialExpressionNamedRerouteUsage_99, Density=@MaterialExpressionComponentMask_34) {
   code="float HeightDensity = Density;\r\n\r\nif(NormAltitude<0.05 || NormAltitude>0.95)\r\n{\r\n   HeightDensity = 0.0;\r\n}\r\n\r\nreturn HeightDensity;"
   inputs [
    {inputName=NormAltitude, input {expression=&self:MaterialExpressionNamedRerouteUsage_99}}
    {inputName=Density, input {expression=&self:MaterialExpressionComponentMask_34}}
   ]
  }
  n107 = Subtract(A=@n106, B=ScalarParameter() {defaultValue=0.681, group=StormDFs, parameterName=StormDonut DF Center})
  n108 = Multiply(A=Divide(A=LinearInterpolate(A=@n79, B=@n80, Alpha=@n83)) {constB=100}, B=Add(A=Add(A=@n93, B=@n104), B=Multiply(A=@n105, B=@n107)))
  n109 = OneMinus(None=Max(A=NamedRerouteUsage(), B=CollectionParameter() {collection=&MPC_SafeZoneClouds, parameterName=PreventStormShrinking}))
  n110 = ReflectionCapturePassSwitch(
   Default=Multiply(A=CollectionParameter() {collection=&MPC_SafeZoneClouds, parameterName=EnableVolumetricStorm}, B=Saturate(None=@n109))
   Reflection=Constant()
  )
  NamedRerouteDeclaration(None=Multiply(A=NamedRerouteDeclaration(None=@n108) {name=CloudMaskSDF, nodeColor=<rgba 0 1 0.3906 1>}, B=@n110)) {name=StormEXT_1, nodeColor=<rgba 1 0 0.8894 1>}
  n111 = CollectionParameter() {collection=&MPC_SafeZoneClouds, parameterName=SafeZoneLoc}
  NamedRerouteDeclaration(None=Subtract(A=WorldPosition().XYZ, B=ComponentMask(None=@n111) {b=true, g=true, r=true})) {name=Safe Zone Local_1, nodeColor=<rgba 1 0 0.7266 1>}
  NamedRerouteDeclaration(None=ScalarParameter() {defaultValue=-0.1, parameterName=LowAltRemapAlt}) {name=LowRemap, nodeColor=<rgba 0.4297 0 1 1>}
  NamedRerouteDeclaration(None=ScalarParameter() {defaultValue=1.5, parameterName=HIghAltRemapAlt}) {name=HighRemap, nodeColor=<rgba 0.9688 0 1 1>}
  NamedRerouteDeclaration(None=ScalarParameter() {defaultValue=0.001, desc=0.002, parameterName=BaseCloudSpeed}) {name=CloudSpeed, nodeColor=<rgba 1 0.4453 0 1>}
  n112 = WorldPosition()
  n113 = MaterialFunctionCall(Km=ScalarParameter() {defaultValue=5, group=Lightning, parameterName=Lighting Pattern Scale}) {materialFunction=&Km-to-Cm}
  n114 = AppendVector(
   A=ComponentMask(None=Divide(A=@n112.XYZ, B=@n113.Cm)) {g=true, r=true}
   B=MaterialFunctionCall(Speed=ScalarParameter() {defaultValue=0.2, group=Lightning, parameterName=Lighting Speed}) {materialFunction=&TimeWithSpeedVariable}.Time
  )
  n115 = TextureSample!(UVs=@n114, Tex, Apply View MipBias) {texture=/Volumetrics/Content/VolumeTextures/Textures/VT_Lightning}
  n116 = ScalarParameter() {defaultValue=3, group=Lightning, parameterName=Lightning Power}
  n117 = Multiply(A=@n115.R, B=@n115.R)
  n118 = ScalarParameter() {group=Lightning, parameterName=Fill Lightning Scatter}
  n119 = ScalarParameter() {defaultValue=2.5, group=Lightning, parameterName=Lightning Mip}
  n120 = Add(
   A=Add(A=Power(Base=@n115.R, Exp=@n116), B=Multiply(A=@n117, B=@n118))
   B=Multiply(
    A=TextureSample!(UVs=@n114, Tex, Level=@n119, Apply View MipBias) {mipValueMode=TMVM_MipLevel, texture=&VT_Lightning}.R
    B=ScalarParameter() {defaultValue=0.05, group=Lightning, parameterName=Lightning Spread}
   )
  ) {constB=0}
  n121 = Power(Base=NamedRerouteUsage(), Exp=ScalarParameter() {defaultValue=1, group=Lightning, parameterName=Lightning Mask Power})
  NamedRerouteDeclaration(None=Multiply(
    A=Multiply(A=@n120, B=@n121)
    B=VectorParameter() {group=Lightning, parameterName=Lightning Color, defaultValue=<rgba 0.0997 0.3491 2 0>}.RGB
   )) {name=Lightning_1, nodeColor=<rgba 1 0 0.2223 1>}
  NamedRerouteDeclaration(None=@n110) {name=Storm Control_1, nodeColor=<rgba 1 0 0.7734 1>}
  n122 = Divide(
   A=Subtract(A=NamedRerouteUsage(), B=ScalarParameter() {group=Storm-Clouds Blending, parameterName=StormBlendingSub})
   B=ScalarParameter() {defaultValue=1, group=Storm-Clouds Blending, parameterName=StormBlendingDiv}
  ) {constB=1}
  NamedRerouteDeclaration(None=Multiply(A=Saturate(None=@n122), B=NamedRerouteUsage())) {name=Storm Cloud Blend, nodeColor=<rgba 0 1 0.8359 1>}
  n123 = NamedRerouteUsage()
  n124 = NamedRerouteUsage()
  n125 = TextureSample(
   UVs=MaterialFunctionCall(X=Saturate(None=@n123), Y=OneMinus(None=@n124)) {materialFunction=&MakeFloat2}.Result
   Level=CollectionParameter() {collection=&MPC_SafeZoneClouds, parameterName=StormProfileMip}
  ) {mipValueMode=TMVM_MipLevel, samplerType=SAMPLERTYPE_LinearColor, texture=/Game/TimeOfDay/CloudsAndStorm/Textures/StormProfileLegacy}
  NamedRerouteDeclaration(None=@n125.A) {name=StormShape_1, nodeColor=<rgba 1 0.141 0.7007 1>}
  n126 = ScalarParameter() {defaultValue=18.6666, group=DebugStorm, parameterName=DebugStormThickness}
  n127 = ScalarParameter() {defaultValue=1.7098, group=DebugStorm, parameterName=DebugStormRadius}
  n128 = ScalarParameter() {defaultValue=0.0533, group=DebugStorm, parameterName=DebugStormAltitude}
  n129 = ScalarParameter() {defaultValue=1.2054, group=DebugStorm, parameterName=DebugStormHeight}
  n130 = MaterialFunctionCall(
   X=CollectionParameter() {collection=&MPC_SafeZoneClouds, parameterName=StormThicknessKm-End}
   Y=CollectionParameter() {collection=&MPC_SafeZoneClouds, parameterName=StormRadiusKm-End}
   Z=CollectionParameter() {collection=&MPC_SafeZoneClouds, parameterName=StormAltitude-End}
   A=CollectionParameter() {collection=&MPC_SafeZoneClouds, parameterName=StormHeight-End}
  ) {materialFunction=&MakeFloat4}
  n131 = CollectionParameter() {collection=&MPC_SafeZoneClouds, parameterName=StormRadiusKm}
  n132 = ScalarParameter() {defaultValue=3, parameterName=StormRadius_PreStartOffset}
  n133 = Saturate(None=NamedRerouteUsage())
  n134 = Multiply(A=@n133, B=@n133)
  n135 = MaterialFunctionCall(
   X=CollectionParameter() {collection=&MPC_SafeZoneClouds, parameterName=StormThicknessKm}
   Y=Add(A=NamedRerouteDeclaration(None=@n131) {name=Radius, nodeColor=<rgba 0 0.8828 1 1>}, B=Multiply(A=@n132, B=@n134))
   Z=CollectionParameter() {collection=&MPC_SafeZoneClouds, parameterName=StormAltitude}
   A=CollectionParameter() {collection=&MPC_SafeZoneClouds, parameterName=StormHeight}
  ) {materialFunction=&MakeFloat4}
  n136 = NamedRerouteDeclaration(None=StaticSwitch(
    True=ScalarParameter() {group=DebugStorm, parameterName=DebugSafeZoneScale}
    False=CollectionParameter() {collection=&MPC_SafeZoneClouds, parameterName=SafeZoneScale}
    Value=NamedRerouteUsage()
   )) {name=SafeZoneScale, nodeColor=<rgba 0.3125 0 1 1>}
  NamedRerouteDeclaration(None=StaticSwitch(
    True=MaterialFunctionCall(X=@n126, Y=@n127, Z=@n128, A=@n129) {materialFunction=&MakeFloat4}.Result
    False=LinearInterpolate(A=@n130.Result, B=@n135.Result, Alpha=@n136)
    Value=NamedRerouteUsage()
   )) {name=Storm Collection Params_1, nodeColor=<rgba 0.4708 0 1 1>}
  n137 = NamedRerouteUsage()
  n138 = Subtract(A=CloudSampleAttribute().NormAltitudeInLayer, B=ComponentMask(None=@n137) {b=true})
  n139 = Add(
   A=ComponentMask(None=@n137) {a=true}
   B=Multiply(A=ScalarParameter() {defaultValue=-0.0857, group=StormRadialWaves, parameterName=WavesHeight}, B=@n76)
  )
  NamedRerouteDeclaration(None=Saturate(None=Divide(A=@n138, B=@n139))) {name=Linear Height Profile_1, nodeColor=<rgba 1 0.1875 0 1>}
  NamedRerouteDeclaration(None=@n125.RGB) {name=Mapped Storm Profile_1, nodeColor=<rgba 0.8984 0 1 1>}
  NamedRerouteDeclaration(None=@n60) {name=worldUVStormStatic, nodeColor=<rgba 0 0.9062 1 1>}
  n140 = NamedRerouteUsage()
  n141 = OneMinus(None=Saturate(None=NamedRerouteUsage()))
  n142 = ScalarParameter() {defaultValue=0.8, group=Storm, parameterName=AOSub}
  n143 = Add(
   A=Divide(A=Subtract(A=@n141, B=@n142), B=ScalarParameter() {defaultValue=0.2, group=Storm, parameterName=AODiv})
   B=OneMinus(None=NamedRerouteUsage())
  )
  n144 = NamedRerouteUsage()
  n145 = Divide(
   A=Subtract(A=Saturate(None=@n144), B=ScalarParameter() {defaultValue=-1, group=Clouds, parameterName=AOCloudsSub})
   B=ScalarParameter() {defaultValue=2, group=Clouds, parameterName=AOCloudsDiv}
  )
  n146 = Add(A=@n145, B=DotProduct(A=SkyAtmosphereLightDirection(), B=Constant3Vector() {constant=<rgba 0 0 1 0>}))
  n147 = Saturate(None=Min(A=Add(A=@n140, B=@n143), B=Saturate(None=@n146)))
  NamedRerouteDeclaration(None=StaticSwitchParameter(True=ReflectionCapturePassSwitch(Default=@n147, Reflection=@n145), False=@n145) {defaultValue=true, group=Storm, parameterName=EnableStorm}) {name=NewAmbientOcclusion, nodeColor=<rgba 0 1 0.0625 1>}
  NamedRerouteDeclaration(None=Multiply(A=NamedRerouteUsage(), B=NamedRerouteUsage())) {name=ch4_clouds, nodeColor=<rgba 0 1 0.7422 1>}
  NamedRerouteDeclaration(None=StaticSwitch(
    True=ScalarParameter() {group=DebugStorm, parameterName=DebugCloudStartingBias}
    False=CollectionParameter() {collection=&MPC_SafeZoneClouds, parameterName=CloudStartingBias}
    Value=NamedRerouteUsage()
   )) {name=CloudStartingBias, nodeColor=<rgba 1 0.2109 0 1>}
  n148 = NamedRerouteUsage()
  n149 = Saturate(None=NamedRerouteUsage())
  n150 = OneMinus(None=@n149)
  n151 = NamedRerouteDeclaration(None=Multiply(A=Ceil(None=@n149), B=Ceil(None=@n150))) {name=ConservativeStorm, nodeColor=<rgba 0.0781 1 0 1>}
  NamedRerouteDeclaration(None=MaterialFunctionCall(X=Max(A=@n148, B=@n151), Y=NamedRerouteUsage()) {materialFunction=&MakeFloat4}.Result) {name=ConservativeDensity2, nodeColor=<rgba 0.2891 0 1 1>}
  NamedRerouteDeclaration(None=StaticBoolParameter() {defaultValue=true, group=DebugStorm, parameterName=UseStormDebugValues}) {name=UseStormDebugValues, nodeColor=<rgba 1 0 0.9141 1>}
  NamedRerouteDeclaration(None=@n104) {name=FlowMap2, nodeColor=<rgba 0.7578 1 0 1>}
  n152 = Power(Base=NamedRerouteUsage(), Exp=ScalarParameter() {defaultValue=1.85, parameterName=AltitudeCurveExponent})
  NamedRerouteDeclaration(None=LinearInterpolate(
    A=ScalarParameter() {defaultValue=0.05, parameterName=DensityLow}
    B=ScalarParameter() {defaultValue=0.05, parameterName=DensityHigh}
    Alpha=NamedRerouteDeclaration(None=@n152) {name=AltitudeMap, nodeColor=<rgba 0.9453 1 0 1>}
   )) {name=AltitudeDensity, nodeColor=<rgba 0.2891 0 1 1>}
  n153 = Multiply(A=ComponentMask(None=NamedRerouteUsage()) {g=true, r=true}, B=ScalarParameter() {defaultValue=10, parameterName=NoiseMaskTilie})
  NamedRerouteDeclaration(None=Power(
    Base=TextureSample!(UVs=@n153, Tex, Apply View MipBias) {samplerType=SAMPLERTYPE_Grayscale, texture=&LowResBlurredNoise}.R
    Exp=ScalarParameter() {defaultValue=0.1, parameterName=Exp_1}
   )) {name=NoiseMask, nodeColor=<rgba 0 1 0.7656 1>}
  NamedRerouteDeclaration(None=MaterialFunctionCall(
    Color_Morning=VectorParameter() {group=Clouds, parameterName=AlbedoCloudsMorning, defaultValue=<rgba 1 1 1 0>}.RGB
    Color_Day=VectorParameter() {group=Clouds, parameterName=AlbedoCloudsDay, defaultValue=<rgba 1 1 1 0>}.RGB
    Color_Evening=VectorParameter() {group=Clouds, parameterName=AlbedoCloudsEvening, defaultValue=<rgba 1 1 1 0>}.RGB
    Color_Night=VectorParameter() {group=Clouds, parameterName=AlbedoCloudsNight, defaultValue=<rgba 1 1 1 0>}.RGB
   ) {materialFunction=&Color_Time_of_Day}.Result) {name=CloudColor, nodeColor=<rgba 0.2422 1 0 1>}
  NamedRerouteUsage()
  NamedRerouteUsage()
  NamedRerouteUsage()
  NamedRerouteUsage()
  NamedRerouteUsage()
  NamedRerouteUsage()
  n154 = Subtract(A=NamedRerouteUsage(), B=@n39)
  OneMinus(None=Saturate(None=Abs(None=@n154)))
  n155 = Subtract(
   A=ComponentMask(None=WorldPosition().XYZ) {g=true, r=true}
   B=ComponentMask(None=CollectionParameter() {collection=&MPC_SafeZoneClouds, parameterName=SafeZoneLoc}) {g=true, r=true}
  )
  n156 = MaterialFunctionCall(Cm=MaterialFunctionCall(Vector 2=NamedRerouteDeclaration(None=@n155) {name=Safe Zone Local_2, nodeColor=<rgba 1 0 0.7266 1>}) {materialFunction=/Engine/Functions/Engine_MaterialFunctions02/Utility/VectorLength}.V2 Length) {materialFunction=&Cm-to-Km}
  n157 = NamedRerouteUsage()
  n158 = ScalarParameter() {defaultValue=1.7928, parameterName=Radius}
  OneMinus(None=Saturate(None=SphereMask(A=@n156.Km, B=@n157, Radius=@n158)))
  OneMinus(None=Saturate(None=NamedRerouteUsage()))
  Power(Base=@n19.R, Exp=ScalarParameter() {defaultValue=0.65, parameterName=Exp})
  Saturate(None=@n1)
  ScalarParameter() {defaultValue=0.064, parameterName=density}
  ScalarParameter() {defaultValue=2250000, parameterName=worldScale}
  VectorParameter() {group=Clouds, parameterName=AlbedoClouds, defaultValue=<rgba 1 1 1 0>}
  n159 = ScalarParameter() {defaultValue=0.2, group=Scattering and Phase, parameterName=MS Occ Morning}
  n160 = ScalarParameter() {defaultValue=0.2, group=Scattering and Phase, parameterName=MS Occ Day}
  n161 = ScalarParameter() {defaultValue=0.2, group=Scattering and Phase, parameterName=MS Occ Evening}
  n162 = ScalarParameter() {defaultValue=0.2, group=Scattering and Phase, parameterName=MS Occ Night}
  VolumetricAdvancedMaterialOutput!(
   PhaseG=NamedRerouteDeclaration(None=ScalarParameter() {defaultValue=0.5, group=Scattering and Phase, parameterName=Phase A}) {name=PhaseG, nodeColor=<rgba 1 0 0.0937 1>}
   PhaseG2=NamedRerouteDeclaration(None=ScalarParameter() {defaultValue=0.8, group=Scattering and Phase, parameterName=Phase B}) {name=PhaseG2, nodeColor=<rgba 1 0 0.0938 1>}
   PhaseBlend=NamedRerouteDeclaration(None=ScalarParameter() {defaultValue=0.4, group=Scattering and Phase, parameterName=Phasey Phasey}) {name=PhaseBlend, nodeColor=<rgba 1 0 0.0938 1>}
   MultiScatteringContribution=NamedRerouteDeclaration(None=ScalarParameter() {defaultValue=0.5, group=Scattering and Phase, parameterName=MS C}) {name=MultiScatteringContribution, nodeColor=<rgba 1 0 0.0938 1>}
   MultiScatteringOcclusion=NamedRerouteDeclaration(None=MaterialFunctionCall(Value_Morning=@n159, Value_Day=@n160, Value_Evening=@n161, Value_Night=@n162) {materialFunction=&Scalar_Time_of_Day}.Result) {name=MultiScatteringOcclusion, nodeColor=<rgba 1 0 0.0938 1>}
   MultiScatteringEccentricity=NamedRerouteDeclaration(None=ScalarParameter() {defaultValue=0.5, group=Scattering and Phase, parameterName=MS Ec}) {name=MultiScatteringEccentricity, nodeColor=<rgba 1 0 0.0938 1>}
   ConservativeDensity=NamedRerouteUsage()
  ) {multiScatteringApproximationOctaveCount=1}
  n163 = ScalarParameter() {defaultValue=0.7, parameterName=CloudConservativeExpandSkip}
  n164 = Multiply(A=Multiply(A=Subtract(A=@n12, B=@n163)) {constB=-1}) {constB=100}
  VolumetricCloudEmptySpaceSkippingOutput!(ContainsMatter=Saturate(None=Sign(None=@n164)))
 ]
}
