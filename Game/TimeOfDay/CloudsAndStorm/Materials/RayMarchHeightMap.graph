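# Material function that ray marches a heightmap, optionally a flipbook frame, toward a world-space light to compute volumetric self-shadowing.
# A box intersection caps the step count, then custom HLSL accumulates thickness through the height field as a mirrored volume, interpolating crossings, with optional blue-noise temporal jitter.
# Beer's law converts thickness to Light Energy.
# Other outputs give raw height, SubUVs and a step-complexity debug color.
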MaterialFunction /Game/TimeOfDay/CloudsAndStorm/Materials/RayMarchHeightMap
graph {
 properties {description=This function Ray Marches a heightmap to return shadow density., libraryCategoriesText [Density]}
 nodes [
  n1 = FunctionInput(Preview=TextureCoordinate()) {
   bUsePreviewValueAsDefault=true
   description="The UVs to use. If this is for a flipbook, plug in the flipbook UVs here."
   inputName=UVs
   inputType=FunctionInput_Vector2
   sortPriority=1
  }
  n2 = Frac(None=@n1)
  n3 = FunctionInput() {
   bUsePreviewValueAsDefault=true
   description="For non SubUV textures leave this value at 1, otherwise set this to equal the number of X,Y frames."
   inputName=SubUV Frames
   inputType=FunctionInput_Vector2
   sortPriority=8
   previewValue=<1 1 0 1>
  }
  n4 = FunctionInput() {
   bUsePreviewValueAsDefault=true
   description="This is the 0-1 animation phase of the SubUV, if any."
   inputName=Animation Phase
   inputType=FunctionInput_Scalar
   sortPriority=9
  }
  n5 = Add(
   A=Divide(A=@n1, B=@n3) {desc=SubUV frame 0}
   B=Divide(
    A=MaterialFunctionCall!(1D Index=@n4, Cells XY=@n3) {materialFunction=/Engine/Functions/Engine_MaterialFunctions01/Coordinates/1Dto2DIndex}.2D Index
    B=@n3
   ) {desc=UpperLeft Corner}
  ) {desc=SubUV per Frame}
  n6 = TextureObject() {samplerType=SAMPLERTYPE_LinearColor, texture=/Engine/Functions/Engine_MaterialFunctions02/ExampleContent/Textures/SphereRenderHeightMap}
  MaterialExpressionFunctionInput_2 = FunctionInput(Preview=@n6) {description=The texture containing the heightmap to ray march., inputName=Texture Object, inputType=FunctionInput_Texture2D}
  n7 = TextureSample!(UVs=@n5, Tex=@MaterialExpressionFunctionInput_2, Level=NamedRerouteUsage(), Apply View MipBias) {mipValueMode=TMVM_MipLevel, samplerType=SAMPLERTYPE_LinearColor}
  MaterialExpressionFunctionInput_0 = FunctionInput() {bUsePreviewValueAsDefault=true, inputName=Heightmap Channel, inputType=FunctionInput_Vector4, sortPriority=11, previewValue=<1 0 0 0>}
  n8 = DotProduct(A=AppendVector(A=@n7.RGB, B=@n7.RGB), B=@MaterialExpressionFunctionInput_0)
  n9 = FunctionInput() {
   bUsePreviewValueAsDefault=true
   description=Starting directly at the surface of the heightmap can cause more artifacts and does not look very volumetric. Starting the trace a bit below the surface improves quality and approximates an integration of multiple nested density/lighting deep much better.
   inputName=Start Bias
   inputType=FunctionInput_Scalar
   sortPriority=5
   previewValue=<0.05 0 0 1>
  }
  MaterialExpressionOneMinus_0 = OneMinus(None=@n9)
  MaterialExpressionMultiply_4 = Multiply(A=@n8, B=@MaterialExpressionOneMinus_0) {desc="Bias Method 2 - No artifacts but does not bias the  edge values"}
  n10 = Multiply(A=Add(A=@MaterialExpressionMultiply_4) {desc=1 to 2}) {constB=0.499, desc=0.5 to 1}
  n11 = VectorParameter() {parameterName=LV, defaultValue=<rgba 0 1 0 0>}
  n12 = FunctionInput(Preview=Multiply(A=ComponentMask(None=@n11.RGB) {b=true, g=true, r=true}) {constB=-1}) {
   bUsePreviewValueAsDefault=true
   description=The Light Vector to trace towards in World Space. The vector will be transformed into tangent space inside of the function.
   inputName=Light Vector WorldSpace
   sortPriority=2
  }
  n13 = Normalize(VectorInput=@n12)
  n14 = Constant3Vector()
  n15 = FunctionInput() {
   bUsePreviewValueAsDefault=true
   description="This defaults to 1 which means the function will trace across 1 UV width. Using smaller values can give higher quality results if you do not need to trace the entire UV width, such as for very dense smoke where the shadows would accumulate to black more quickly."
   inputName=Trace Distance
   inputType=FunctionInput_Scalar
   sortPriority=7
   previewValue=<1 0 0 1>
  }
  n16 = Multiply(
   A=Subtract(
    A=Transform(None=@n13) {transformSourceType=TRANSFORMSOURCE_World, transformType=TRANSFORM_Tangent}
    B=FunctionInput(Preview=@n14) {bUsePreviewValueAsDefault=true, inputName=ScreenSpaceTraceOffset}
   )
   B=@n15
  ) {desc=Raw Trace Vec}
  n17 = FunctionInput() {
   bUsePreviewValueAsDefault=true
   description=The max number of steps to trace. The tracer will exit early if it crosses the UV border.
   inputName=Max Steps
   inputType=FunctionInput_Scalar
   sortPriority=3
   previewValue=<64 0 0 1>
  }
  n18 = Divide(B=@n17) {desc=StepSize}
  n19 = Divide(
   A=MaterialFunctionCall!(Ray Origin=AppendVector(A=@n2, B=@n10), Ray Direction=Multiply(A=@n16), Box Min=Constant(), Box Max=Constant() {r=1}) {materialFunction=/Engine/Functions/Engine_MaterialFunctions01/Math/BoxIntersection}.T1
   B=@n18
  ) {desc=How many Steps Fit to edge}
  Ceil(None=@n19)
  Divide(B=@n3) {desc=Cell Size}
  MaterialExpressionFrac_1 = Frac(None=@n5) {desc=Last Step Saturates so must be 0-1}
  FunctionOutput(None=@MaterialExpressionFrac_1) {outputName=SubUVs}
  MaterialExpressionMin_0 = Min(A=@n17, B=Floor(None=@n19) {desc=Run 1 less step})
  MaterialExpressionMultiply_1 = Multiply(A=Multiply(A=@n18) {desc=raw step size}, B=@n15) {desc=Step Size}
  n20 = MaterialFunctionCall(Float2=@n3) {materialFunction=&BreakOutFloat2Components}
  n21 = Max(A=@n20.R, B=@n20.R)
  MaterialExpressionDivide_8 = Divide(
   A=Multiply(A=Multiply(A=@n18, B=@n16) {desc=Scaled TraceVec}) {desc=Trace Vec}
   B=MaterialFunctionCall(X=@n21, Y=@n21, Z=Constant() {r=1}) {materialFunction=&MakeFloat3}.Result
  )
  MaterialExpressionFunctionInput_20 = FunctionInput() {
   bUsePreviewValueAsDefault=true
   description=This setting lets you scale the height of the heightmap.
   inputName=Height Scale
   inputType=FunctionInput_Scalar
   sortPriority=6
   previewValue=<1 0 0 1>
  }
  MaterialExpressionStaticSwitch_0 = StaticSwitch(
   True=Constant() {r=1}
   False=Constant()
   Value=FunctionInput(Preview=StaticBool()) {bUsePreviewValueAsDefault=true, inputName=Temporal Jitter, inputType=FunctionInput_StaticBool, sortPriority=10}
  )
  MaterialExpressionScalarParameter_16 = ScalarParameter() {parameterName=Start Threshold}
  MaterialExpressionNamedRerouteDeclaration_0 = NamedRerouteDeclaration(None=FunctionInput() {bUsePreviewValueAsDefault=true, inputName=Mip Level, inputType=FunctionInput_Scalar, sortPriority=13}) {name=mip, nodeColor=<rgba 0 1 0.0391 1>}
  MaterialExpressionMaterialFunctionCall_0 = MaterialFunctionCall!(Alpha Threshold, Random) {materialFunction=/Game/TimeOfDay/CloudsAndStorm/Materials/MaterialFunctions/DitherTemporalAASpatioTemporalBlue}
  n22 = Custom!(
   Tex=@MaterialExpressionFunctionInput_2
   startZ=@MaterialExpressionMultiply_4
   UV=@MaterialExpressionFrac_1
   NumSteps=@MaterialExpressionMin_0
   StepSize=@MaterialExpressionMultiply_1
   TraceVec=@MaterialExpressionDivide_8
   bias=@MaterialExpressionOneMinus_0
   heightscale=@MaterialExpressionFunctionInput_20
   TemporalJitter=@MaterialExpressionStaticSwitch_0
   threshold=@MaterialExpressionScalarParameter_16
   channel=@MaterialExpressionFunctionInput_0
   mip=@MaterialExpressionNamedRerouteDeclaration_0
   InBlueNoise=@MaterialExpressionMaterialFunctionCall_0.Result
  ) {
   code="//////////////// Ray March HeightMap Function ///////////////////////////\r\n\r\n//** Tex               **// Input Texture Object\r\n//** startZ            **// Input float for start height of texture\r\n//** UV                **// Input float2 containing UVs\r\n//** NumSteps          **// Input float for Max Steps\r\n//** StepSize          **// Input float for ray step size\r\n//** TraceVec          **// Input float3 for TraceVector, scaled by StepSize\r\n//** FinalStepTraceVec **// Input float3 for TraceVector, scaled by StepSize\r\n//** FinalStepSize     **// Input float for length of the final step\r\n//** bias              **// Input float for ray starting bias\r\n//** heightscale       **// Input float for heightmap scale\r\n//** TemporalJitter    **// Input bool (using int) to toggle Temporal Jitter\r\n//** threshold         **// Input float for particle density\r\n\r\nif(startZ<=threshold) return 0;\r\n\r\nfloat TimeLerp = 1;\r\nfloat DepthDiff = 0;\r\nfloat LastDiff = -bias;\r\n\r\n//We scale Z by 2 since the heightmap represents two halves of as symmetrical volume texture, split along Z where texture = 0\r\nfloat3 RayStepUVz = float3(TraceVec.x, TraceVec.y, TraceVec.z*2);\r\n\r\nfloat accum = 0;\r\nfloat3 RayUVz = float3(UV, (startZ) );\r\n\r\n\r\nif(TemporalJitter)\r\n{\r\n\t// jitter the starting position\r\n\tint3 randpos = int3(Parameters.SvPosition.xy, View.StateFrameIndexMod8);\r\n\tfloat rand = float(Rand3DPCG16(randpos).x) / 0xffff;\r\n\tRayUVz += RayStepUVz * InBlueNoise;\r\n}\r\n\r\nint i = 0;\r\nwhile (i < NumSteps)\r\n{\r\n\r\n\tRayUVz += RayStepUVz;\r\nRayUVz.xy = saturate(RayUVz.xy);\r\n\tfloat SampleDepth = dot( channel, Tex.SampleLevel(TexSampler, RayUVz.xy,mip)*heightscale );\r\n\tDepthDiff = abs(RayUVz.z) - abs(SampleDepth);\r\n\r\n\r\n\t\tif (DepthDiff <= 0)\r\n\t\t{\r\n\r\n\t\t\tif(LastDiff > 0)\r\n\t\t\t{\r\n\t\t\t\tTimeLerp = saturate( LastDiff / (LastDiff - DepthDiff));\r\n\t\t\t\taccum += StepSize * (1-TimeLerp);\r\n\t\t\t\t//accum+=StepSize;\r\n\t\t\t}\r\n\t\t\telse\r\n\t\t\t{\r\n\t\t\t\taccum+=StepSize;\r\n\t\t\t}\r\n\t\t}\r\n\t\telse\r\n\t\tif(LastDiff <= 0)\r\n\t\t{\r\n\t\t\tTimeLerp = saturate( LastDiff / (LastDiff - DepthDiff));\r\n\t\t\taccum += StepSize * (TimeLerp);\r\n\t\t\t//accum+=StepSize;\r\n\t\t}\r\n\r\n\tLastDiff = DepthDiff;\r\n\r\n\ti++;\r\n}\r\n\r\n\r\n//Run one more iteration outside of the loop. Using the Box Intersection in the material graph, we precompute the number of whole steps that can be run which leaves one final 'short step' which is the remainder that is traced here. This was cheaper than checking or clamping the UVs inside of the loop.\r\nRayUVz += RayStepUVz;\r\nRayUVz.xy = saturate(RayUVz.xy);\r\nfloat SampleDepth = dot( channel, Tex.SampleLevel(TexSampler, RayUVz.xy,mip)*heightscale );\r\nDepthDiff = abs(RayUVz.z) - abs(SampleDepth);\r\n\r\n\r\n\tif (DepthDiff <= 0)\r\n\t{\r\n\t\taccum+=StepSize;\r\n\t}\r\n\telse\r\n\tif(LastDiff <= 0)\r\n\t{\r\n\t\tTimeLerp = saturate( LastDiff / (LastDiff - DepthDiff));\r\n\t\taccum += StepSize * (TimeLerp);\r\n}\r\n\r\nreturn accum;"
   description=RayMarch Heightmap
   outputType=CMOT_Float1
   inputs [
    {inputName=Tex, input {expression=&self:MaterialExpressionFunctionInput_2}}
    {inputName=startZ, input {expression=&self:MaterialExpressionMultiply_4}}
    {inputName=UV, input {expression=&self:MaterialExpressionFrac_1}}
    {inputName=NumSteps, input {expression=&self:MaterialExpressionMin_0}}
    {inputName=StepSize, input {expression=&self:MaterialExpressionMultiply_1}}
    {inputName=TraceVec, input {expression=&self:MaterialExpressionDivide_8}}
    {inputName=bias, input {expression=&self:MaterialExpressionOneMinus_0}}
    {inputName=heightscale, input {expression=&self:MaterialExpressionFunctionInput_20}}
    {inputName=TemporalJitter, input {expression=&self:MaterialExpressionStaticSwitch_0}}
    {inputName=threshold, input {expression=&self:MaterialExpressionScalarParameter_16}}
    {inputName=channel, input {expression=&self:MaterialExpressionFunctionInput_0}}
    {inputName=mip, input {expression=&self:MaterialExpressionNamedRerouteDeclaration_0}}
    {
     inputName=InBlueNoise
     input {expression=&self:MaterialExpressionMaterialFunctionCall_0, outputIndex=5}
    }
   ]
  }
  n23 = FunctionInput() {
   bUsePreviewValueAsDefault=true
   description=This controls the density of the shadow result.
   inputName=Shadow Density
   inputType=FunctionInput_Scalar
   sortPriority=4
   previewValue=<8 0 0 1>
  }
  FunctionOutput(None=MaterialFunctionCall!(Thickness=@n22, Depth Scale=@n23) {materialFunction=/Engine/Functions/Engine_MaterialFunctions01/Volumetrics/BeersLaw}.Result) {outputName=Light Energy}
  MaterialExpressionMultiply_0 = Multiply(A=@n17)
  n24 = Custom!(Steps=@MaterialExpressionMin_0, MaxSteps=@MaterialExpressionMultiply_0) {
   code="float3 colors[10];\r\n\r\ncolors[0]=float3(0,1,0.127);\r\ncolors[1]=float3(0,1,0);\r\ncolors[2]=float3(0.046,0.52,0);\r\ncolors[3]=float3(0.215,0.215,0);\r\ncolors[4]=float3(0.52,0.046,0);\r\ncolors[5]=float3(0.7,0,0);\r\ncolors[6]=float3(1,0,0);\r\ncolors[7]=float3(1,0,0.5);\r\ncolors[8]=float3(1,0.9,0.9);\r\ncolors[9]=float3(1,1,1);\r\n\r\nSteps/=MaxSteps-1;\r\nSteps=saturate(Steps);\r\n\r\nfloat cindex = floor((Steps)*8);\r\nfloat cphase = frac((Steps)*8);\r\n\r\nreturn lerp(colors[cindex],colors[cindex+1],cphase);"
   description=ShaderComplexity
   inputs [
    {inputName=Steps, input {expression=&self:MaterialExpressionMin_0}}
    {inputName=MaxSteps, input {expression=&self:MaterialExpressionMultiply_0}}
   ]
  }
  FunctionOutput(None=@n24) {
   description="Step Complexity can show how the cost of this function is dependent on viewing angle and lighting direction and number of steps, as well as the heightmap content."
   outputName=Step Complexity
   sortPriority=2
  }
  FunctionOutput(None=@n8) {description=This is a single lookup used from the specified heightmap texture and channel., outputName=Raw Heightmap Value, sortPriority=1}
  Subtract(A=@n8, B=@n9) {desc=Bias Method 1 - Has Edge Artifacts}
 ]
}
