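# Masked two-sided-foliage material for 2D distant horizon clouds, lit from the SkyAtmosphere light.
# RayMarchHeightMap traces a 4x4 cloud flipbook, phased by UV1, with Voronoi noise and blue-noise dither jittering the trace.
# Its light energy combines with sky luminance, aerial perspective, dusk-dependent terms and a curl-noise-distorted resample driving subsurface highlights.
# Raw height, scaled by density, feeds a temporal dither for soft masked edges.
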Material /Game/TimeOfDay/CloudsAndStorm/Materials/M_2D_DistantHorizonClouds
graph {
 properties {
  bApplyCloudFogging=true
  bCastRayTracedShadows=false
  bComputeFogPerPixel=true
  bTangentSpaceNormal=false
  bUseLightmapDirectionality=false
  blendMode=BLEND_Masked
  opacityMaskClipValue=0.45
  pixelDepthOffsetMode=PDOM_Legacy
  shadingModel=MSM_TwoSidedFoliage
  translucencyLightingMode=TLM_VolumetricPerVertexNonDirectional
  usedShadingModels=Two Sided Foliage
 }
 nodes [
  n1 = Add(
   A=Multiply(A=TextureCoordinate(), B=ScalarParameter() {defaultValue=1, parameterName=NoiseOFfsetFrequency})
   B=Panner() {speedX=-0.016, speedY=0.006}
  )
  n2 = TextureSample!(UVs=@n1, Tex, Bias=ScalarParameter() {defaultValue=2, parameterName=NoiseMipBias}, Apply View MipBias) {mipValueMode=TMVM_MipBias, samplerType=SAMPLERTYPE_LinearColor, texture=/Game/TimeOfDay/CloudsAndStorm/Textures/VoronoiNoise2D_v01}
  n3 = DotProduct(A=@n2.RGBA, B=Constant4Vector() {desc=fbm weights 0.7, constant=<rgba 0.4494 0.3146 0.1573 0.0787>}.RGBA)
  n4 = AppendVector(A=ComponentMask(None=ConstantBiasScale(None=@n3) {bias=-0.5, scale=2}) {g=true, r=true}, B=Constant())
  n5 = ScalarParameter() {defaultValue=1, parameterName=NoiseScale}
  n6 = MaterialFunctionCall!(Alpha Threshold, Random) {materialFunction=/Game/TimeOfDay/CloudsAndStorm/Materials/MaterialFunctions/DitherTemporalAASpatioTemporalBlue}
  n7 = MaterialFunctionCall(Input=@n6.Result, Target Low=ScalarParameter() {defaultValue=0.3333, parameterName=NoiseNorm}) {materialFunction=&RemapValueRange}
  n8 = Clamp(None=Multiply(A=@n4, B=Multiply(A=@n5, B=@n7.Result))) {minDefault=-1}
  n9 = ComponentMask(None=TextureCoordinate() {coordinateIndex=1}) {r=true}
  n10 = ScalarParameter() {defaultValue=1.5, parameterName=HS}
  n11 = Multiply(A=Multiply(A=Subtract(A=@n6.Result) {constB=0.5}), B=ScalarParameter() {defaultValue=0.2, parameterName=ditherGain})
  n12 = MaterialFunctionCall!(
   ScreenSpaceTraceOffset=@n8
   Texture Object=TextureObject() {samplerType=SAMPLERTYPE_LinearColor, texture=/Game/TimeOfDay/CloudsAndStorm/Textures/Cloud2DCombined_v02}
   UVs
   Light Vector WorldSpace=SkyAtmosphereLightDirection()
   Max Steps=ScalarParameter() {defaultValue=10, parameterName=Steps}
   Shadow Density=ScalarParameter() {defaultValue=6, parameterName=Shadow Density}
   Start Bias=ScalarParameter() {defaultValue=0.5, parameterName=StartBias}
   Height Scale=Add(A=@n10, B=Multiply(A=@n10, B=@n11))
   Trace Distance=ScalarParameter() {defaultValue=0.5, parameterName=TraceDistance}
   SubUV Frames=Constant2Vector() {g=4, r=4}
   Animation Phase=@n9
   Temporal Jitter=StaticBoolParameter() {defaultValue=true, parameterName=Jitter}
   Heightmap Channel
   Mip Level=ScalarParameter() {parameterName=Mip}
  ) {materialFunction=&/RayMarchHeightMap}
  n13 = Multiply(A=TextureCoordinate(), B=ScalarParameter() {defaultValue=5, parameterName=detailNoiseFreq})
  n14 = Multiply(
   A=ComponentMask(None=TextureSample!(UVs=@n13, Tex, Apply View MipBias) {samplerType=SAMPLERTYPE_LinearColor, texture=/Volumetrics/Content/VolumeTextures/Textures/T_TilingSimpleCurlNoise}.RGB) {g=true, r=true}
   B=ScalarParameter() {defaultValue=0.0005, parameterName=DetailSDFNoiseGain}
  )
  n15 = TextureSample(UVs=Subtract(A=@n12.SubUVs, B=@n14)) {samplerType=SAMPLERTYPE_LinearColor, texture=&Cloud2DCombined_v02}
  n16 = ScalarParameter() {defaultValue=0.15, parameterName=HighlightLow}
  n17 = ScalarParameter() {defaultValue=0.5, parameterName=HighlightHigh}
  n18 = ScalarParameter() {defaultValue=0.1, parameterName=SSLow}
  n19 = Saturate(None=Multiply(A=@n15.R) {constB=1.7})
  n20 = OneMinus(None=Multiply(
    A=MaterialFunctionCall(Input=@n15.G, Input Low=@n16, Input High=@n17, Target Low=@n18) {materialFunction=&RemapValueRange}.Result
    B=OneMinus(None=@n19)
   ))
  n21 = ScalarParameter() {defaultValue=1, parameterName=SubSurfaceGain}
  n22 = ScalarParameter() {defaultValue=1, parameterName=Albedo}
  n23 = Multiply(A=OneMinus(None=Power(Base=@n20)), B=Multiply(A=Multiply(A=@n21, B=@n22), B=NamedRerouteUsage()))
  n24 = ComponentMask(None=Multiply(A=SkyAtmosphereAerialPerspective!(World Position), B=ScalarParameter() {defaultValue=1, parameterName=AerialHorizon})) {b=true, g=true, r=true}
  n25 = LinearInterpolate(B=Multiply(A=@n24) {constB=32}, Alpha=NamedRerouteUsage()) {constA=1}
  Add(A=@n23, B=@n25)
  Add(A=@n24, B=Multiply(A=SkyAtmosphereDistantLightScatteredLuminance(), B=ScalarParameter() {parameterName=AtmosHorizon}))
  Add(A=@n5, B=@n7.Result)
  n26 = SkyAtmosphereDistantLightScatteredLuminance()
  n27 = SkyAtmosphereViewLuminance!(WorldDirection)
  Add(A=@n26, B=@n27)
  n28 = MaterialFunctionCall() {materialFunction=&ScreenAlignedUVs}
  Add(A=@n28."X 100%, Y 100%") {constB=0}
  AtmosphericFogColor!(World Position)
  Constant() {r=3}
  Constant()
  Constant()
  n29 = Multiply(A=Saturate(None=Subtract(A=@n12.Raw Heightmap Value) {constB=0.05}), B=ScalarParameter() {defaultValue=3, parameterName=Density})
  n30 = Add(A=ScalarParameter() {parameterName=AlphaEdgeSharpness})
  MaterialFunctionCall(
   Alpha Threshold=Saturate(None=Power(Base=@n29, Exp=@n30) {constExponent=3})
   Random=ScalarParameter() {defaultValue=0.9, parameterName=SpatiotemporalGain}
  ) {materialFunction=&DitherTemporalAASpatioTemporalBlue}
  MaterialFunctionCall(Input=@n15.G, Input Low=@n16, Input High=@n17) {materialFunction=&RemapValue_0-1}
  MaterialFunctionCall() {materialFunction=&Scalar_Time_of_Day}
  Multiply(A=SkyAtmosphereViewLuminance()) {constB=2}
  Multiply(A=@n4, B=@n5)
  Multiply(A=Multiply(A=@n23, B=@n25), B=NamedRerouteUsage())
  n31 = SkyAtmosphereLightDirection()
  Multiply(A=@n31)
  n32 = LinearInterpolate(
   A=ScalarParameter() {defaultValue=1, parameterName=Albedo}
   B=ScalarParameter() {defaultValue=0.4, parameterName=AlbedoDusk}
   Alpha=NamedRerouteUsage()
  )
  n33 = Saturate(None=Power(Base=@n12.Light Energy) {constExponent=1.5})
  n34 = ComponentMask(None=Multiply(A=SkyAtmosphereAerialPerspective(), B=ScalarParameter() {defaultValue=0.5, parameterName=SkyAerialGain})) {b=true, g=true, r=true}
  n35 = NamedRerouteUsage()
  n36 = Multiply(
   A=Multiply(A=LinearInterpolate(A=@n27, B=@n26, Alpha=@n35) {constAlpha=0.3333}, B=ScalarParameter() {defaultValue=3, parameterName=SkyAtmosGain})
   B=Multiply(A=NamedRerouteUsage())
  )
  Multiply(
   A=Add(A=Multiply(A=@n32, B=@n33), B=Add(A=@n34, B=@n36))
   B=VectorParameter() {parameterName=CloudTintHigh, defaultValue=<rgba 0.0544 0.0817 0.1198 1>}.RGB
  )
  NamedRerouteDeclaration(None=@n8) {name=norm, nodeColor=<rgba 0 0.7891 1 1>}
  NamedRerouteDeclaration(None=MaterialFunctionCall!(A=@n28."X 100%, Y 100%", B=Constant() {r=0.5}, Bounds=Constant() {r=0.9}, Edge Falloff=Constant() {r=0.042}) {materialFunction=/Engine/Functions/Engine_MaterialFunctions01/Gradient/BoxMask-2D}.Result) {name=ssfalloff, nodeColor=<rgba 1 0.5156 0 1>}
  NamedRerouteDeclaration(None=@n14) {name=detailNoiseOffset, nodeColor=<rgba 0.0547 0 1 1>}
  n37 = DotProduct(A=Normalize(VectorInput=SkyAtmosphereLightDirection()), B=Constant3Vector() {constant=<rgba 0 0 1 0>})
  n38 = Saturate(None=OneMinus(None=Abs(None=@n37)))
  NamedRerouteDeclaration(None=Power(Base=@n38) {constExponent=12}) {name=DuskAtmosMut, nodeColor=<rgba 0 0.0156 1 1>}
  n39 = SkyAtmosphereLightDirection()
  n40 = Normalize(VectorInput=Multiply(A=Abs(None=@n39), B=Constant3Vector() {constant=<rgba 1 1 0 0>}))
  NamedRerouteDeclaration(None=ConstantBiasScale(None=DotProduct(A=@n39, B=@n40))) {name=whichway, nodeColor=<rgba 0 1 0.1328 1>}
  NamedRerouteDeclaration(None=@n38) {name=DustMultRaw, nodeColor=<rgba 0 0.0156 1 1>}
  n41 = DotProduct(A=VertexNormalWS(), B=SkyAtmosphereLightDirection())
  NamedRerouteDeclaration(None=OneMinus(None=ConstantBiasScale(None=@n41))) {name=SSFallfoffBack, nodeColor=<rgba 0.875 1 0 1>}
  NamedRerouteUsage()
  NamedRerouteUsage()
  n42 = Multiply(A=@n31, B=Constant3Vector() {constant=<rgba 1 1 0 0>})
  n43 = ScalarParameter() {defaultValue=0.6666, parameterName=LightBias}
  Normalize(VectorInput=Add(A=VertexNormalWS(), B=Multiply(A=@n42, B=@n43)))
  OneMinus(None=@n9)
  n44 = Saturate(None=Multiply(A=@n12.Raw Heightmap Value))
  OneMinus(None=Power(Base=OneMinus(None=@n44)) {constExponent=8})
  n45 = Multiply(A=@n15.R) {constB=2.5}
  OneMinus(None=@n45)
  n46 = ComponentMask(None=@n12.SubUVs) {g=true}
  n47 = Subtract(A=@n46) {constB=0.6}
  Power(Base=Saturate(None=OneMinus(None=@n47))) {constExponent=16}
  Saturate(None=Add(A=@n46) {constB=0.2})
  Saturate(None=@n45)
  ScalarParameter() {defaultValue=1, parameterName=AO}
  ScalarParameter() {defaultValue=1, parameterName=HighlightMip}
  ScalarParameter() {bUseCustomPrimitiveData=true, parameterName=AnimFrame}
  Sign(None=@n37)
  Subtract(A=@n1, B=NamedRerouteUsage())
  TextureCoordinate()
  TextureCoordinate()
  TextureObject() {samplerType=SAMPLERTYPE_LinearColor, texture=/Game/TimeOfDay/CloudsAndStorm/Textures/Cloud2D}
  VectorParameter() {parameterName=CloudTintPink, defaultValue=<rgba 0.7604 0.5703 0.6942 1>}
 ]
}
