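# Material function for baked cloud cards in the time-of-day sky, setting emissive color and opacity.
# Sun-direction masks, computed per vertex and passed through interpolators, weight directional lighting channels packed in two textures, faking relightable volume.
# The result blends shadow and light colors from the Fortnite parameter collection, plus height and post-process fog.
# Opacity combines alpha eroded by scrolling noise with skybox and overcast fades; storm darkening nodes don't feed the output.
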MaterialFunction /Game/TimeOfDay/CloudsAndStorm/Materials/MF_BakedClouds
graph {nodes [
  n1 = ComponentMask(None=CollectionParameter() {collection=&FortniteMaterialParameters, parameterName=CloudsLight}) {b=true, g=true, r=true}
  n2 = Desaturation(None=@n1, Fraction=Constant() {r=-0.8})
  n3 = TextureCoordinate()
  n4 = TextureSampleParameter2D(UVs=@n3) {parameterName=Texture A, samplerType=SAMPLERTYPE_LinearColor, texture=/Game/TimeOfDay/CloudsAndStorm/Textures/T_BakedClouds01_A}
  n5 = NamedRerouteUsage()
  n6 = NamedRerouteUsage()
  n7 = NamedRerouteUsage()
  n8 = TextureSampleParameter2D(UVs=@n3) {parameterName=Texture B, samplerType=SAMPLERTYPE_LinearColor, texture=/Game/TimeOfDay/CloudsAndStorm/Textures/T_BakedClouds01_B}
  n9 = NamedRerouteUsage()
  n10 = Add(A=Add(A=Multiply(A=@n4.R, B=@n5), B=Multiply(A=@n4.G, B=@n6)), B=Add(A=Multiply(A=@n4.B, B=@n7), B=Multiply(A=@n8.R, B=@n9)))
  n11 = Add(A=Multiply(A=@n8.G, B=NamedRerouteUsage()), B=Multiply(A=@n8.B, B=NamedRerouteUsage()))
  n12 = StaticSwitchParameter(True=@n4.R, False=LinearInterpolate(A=Add(A=@n10, B=@n11), B=@n8.A, Alpha=NamedRerouteUsage())) {parameterName=SingleMaskDebug}
  n13 = Multiply(A=@n12, B=@n12)
  n14 = OneMinus(None=Multiply(A=@n13, B=@n13))
  n15 = Multiply(A=@n14, B=@n14)
  n16 = NamedRerouteUsage()
  n17 = VertexInterpolator(VS=Add(A=Multiply(A=@n16) {constB=2.5}))
  n18 = Multiply(A=Multiply(A=@n12, B=Multiply(A=@n15, B=@n15)), B=@n17.PS)
  n19 = LinearInterpolate(A=NamedRerouteUsage(), B=Multiply(A=LinearInterpolate(A=@n1, B=@n2, Alpha=@n18), B=@n17.PS), Alpha=@n12)
  n20 = CollectionParameter() {
   collection=&FortniteMaterialParameters
   parameterName=CloudsDirectSunLightColor
  }
  n21 = ComponentMask(None=@n20) {b=true, g=true, r=true}
  n22 = CollectionParameter() {collection=&FortniteMaterialParameters, parameterName=CloudsHeightFog}
  n23 = WorldPosition()
  n24 = Divide(A=@n23.Z) {constB=300000}
  n25 = OneMinus(None=Saturate(None=NamedRerouteDeclaration(None=@n24) {name=Height Mask, nodeColor=<rgba 0.2188 0 1 1>}))
  n26 = Multiply(A=@n25, B=@n25)
  n27 = Multiply(
   A=LinearInterpolate(A=Multiply(A=@n19, B=@n21), B=ComponentMask(None=@n22) {b=true, g=true, r=true}, Alpha=Multiply(A=@n26) {constB=0.6})
   B=ScalarParameter() {defaultValue=4, group=Colors, parameterName=Color Boost}
  ) {constB=4}
  n28 = CollectionParameter() {collection=&FortniteMaterialParameters, parameterName=FogPostProcessColor}
  n29 = ComponentMask(None=@n28) {b=true, g=true, r=true}
  n30 = NamedRerouteUsage()
  FunctionOutput(None=SetMaterialAttributes!(MaterialAttributes, EmissiveColor=Add(A=@n27, B=@n29), Opacity=@n30) {attributeSetTypes [#1, #2]})
  MaterialFunctionCall!(Color=@n27, Storm Color=VectorParameter() {parameterName=Storm Darkening, defaultValue=<rgba 0.0337 0.0445 0.0573 1>}.RGB, Height) {materialFunction=/Game/TimeOfDay/CloudsAndStorm/Materials/MaterialFunctions/MF_Storm_DarkeningCircle}
  n31 = MaterialFunctionCall() {materialFunction=/Game/TimeOfDay/CloudsAndStorm/Materials/MaterialFunctions/MF_RemappedStormScale}
  Multiply(A=@n30, B=Saturate(None=Multiply(A=@n31.StartingBias) {constB=2}))
  n32 = SkyAtmosphereLightDirection()
  n33 = Multiply(A=Normalize(VectorInput=WorldPosition() {worldPositionShaderOffset=WPT_CameraRelative}.XYZ)) {constB=-1}
  n34 = NamedRerouteDeclaration(None=DotProduct(A=@n32, B=@n33)) {name=Vertex Light Dot, nodeColor=<rgba 0.125 1 0 1>}
  n35 = Normalize(VectorInput=CrossProduct(A=@n33, B=Constant3Vector() {constant=<rgba 0 0 1 0>}))
  n36 = Transform(None=Constant3Vector() {constant=<rgba 0 1 0 0>}) {transformSourceType=TRANSFORMSOURCE_Local}
  n37 = CrossProduct(A=Constant3Vector() {constant=<rgba 0 0 -1 0>}, B=@n36)
  n38 = Sign(None=ComponentMask(None=Transform(None=@n37) {transformSourceType=TRANSFORMSOURCE_World, transformType=TRANSFORM_Local}) {r=true})
  n39 = NamedRerouteUsage()
  n40 = Sign(None=Multiply(A=DotProduct(A=@n36, B=@n39)) {constB=-1})
  n41 = Multiply(A=Multiply(A=@n35, B=Multiply(A=@n38, B=@n40)) {constB=-1}) {constB=2}
  n42 = DotProduct(A=@n32, B=Normalize(VectorInput=Add(A=@n33, B=@n41)))
  n43 = DotProduct(A=@n32, B=Normalize(VectorInput=Subtract(A=@n41, B=@n33)))
  n44 = Power(Base=MaterialFunctionCall(R=Abs(None=@n34), G=Abs(None=@n42), B=Abs(None=@n43)) {materialFunction=&AppendMany}.RGB) {constExponent=5}
  n45 = MaterialFunctionCall(Float3=@n44) {materialFunction=&BreakOutFloat3Components}
  n46 = Add(A=@n45.R, B=@n45.R)
  n47 = MaterialFunctionCall(Float3=Divide(A=@n44, B=Add(A=@n46, B=@n45.B))) {materialFunction=&BreakOutFloat3Components}
  n48 = Multiply(A=@n47.R, B=Sign(None=@n34))
  NamedRerouteDeclaration(None=VertexInterpolator(VS=Saturate(None=@n48)).PS) {name=Mask 1, nodeColor=<rgba 1 0 0.2706 1>}
  n49 = Multiply(A=@n48) {constB=-1}
  NamedRerouteDeclaration(None=VertexInterpolator(VS=Saturate(None=@n49)).PS) {name=Mask 4, nodeColor=<rgba 1 0 0.2706 1>}
  n50 = NamedRerouteUsage()
  n51 = ComponentMask(None=@n50) {b=true}
  n52 = Divide(A=Constant() {r=2}, B=Multiply(A=Add(A=@n51) {constB=-0.5}) {constB=0.8})
  n53 = Multiply(A=ComponentMask(None=Multiply(A=@n52, B=@n50)) {g=true, r=true}) {constB=0.7}
  n54 = Multiply(
   A=Time()
   B=AppendVector(
    A=ScalarParameter() {defaultValue=0.02, parameterName=Scroll Speed X}
    B=ScalarParameter() {defaultValue=0.015, parameterName=Scroll Speed Y}
   )
  )
  n55 = TextureSample(UVs=VertexInterpolator(VS=Add(A=@n53, B=@n54)).PS) {samplerType=SAMPLERTYPE_LinearColor, texture=/Game/TimeOfDay/CloudsAndStorm/Textures/T_BakedClouds_Noise}
  NamedRerouteDeclaration(None=OneMinus(None=@n55.R)) {name=Texture Noise, nodeColor=<rgba 1 0.5391 0 1>}
  n56 = NamedRerouteUsage()
  n57 = Power(
   Base=NamedRerouteUsage()
   Exp=LinearInterpolate(A=ScalarParameter() {defaultValue=1, parameterName=Alpha Contraction}, B=LinearInterpolate(Alpha=@n56) {constA=6}, Alpha=@n55.R) {constA=1, constB=6}
  ) {constExponent=8}
  n58 = ScalarParameter() {defaultValue=1, parameterName=Alpha Expansion}
  n59 = Multiply(A=Saturate(None=Multiply(A=@n57, B=@n58)), B=NamedRerouteUsage())
  n60 = NamedRerouteUsage()
  n61 = Constant() {r=-12}
  n62 = Constant() {r=10.55}
  n63 = SquareRoot(None=Saturate(None=CollectionParameter() {collection=&FortniteMaterialParameters, parameterName=OvercastSky}))
  n64 = Multiply(A=Add(A=Add(A=@n60, B=@n61) {constB=-12}, B=Multiply(A=@n62, B=@n63))) {constB=-0.6}
  n65 = CollectionParameter() {collection=&FortniteMaterialParameters, parameterName=SkyBoxClouds}
  n66 = Saturate(None=Multiply(A=Saturate(None=@n64), B=Saturate(None=@n65)))
  n67 = NamedRerouteUsage()
  n68 = Multiply(A=@n67, B=@n67)
  n69 = Multiply(A=Multiply(A=@n68, B=@n68)) {constB=4}
  n70 = Multiply(A=@n66, B=Add(A=@n66, B=Saturate(None=@n69)))
  n71 = Saturate(None=@n70)
  n72 = Multiply(A=@n71, B=@n71)
  NamedRerouteDeclaration(None=Multiply(A=Multiply(A=@n59, B=@n72)) {constB=1.01}) {name=Opacity, nodeColor=<rgba 0 1 0.4844 1>}
  NamedRerouteDeclaration(None=Saturate(None=Divide(A=@n23.Z) {constB=30000})) {name=Bottom Mask, nodeColor=<rgba 0.1016 1 0 1>}
  NamedRerouteDeclaration(None=@n4.G) {name=Front Mask, nodeColor=<rgba 1 0.6094 0 1>}
  n73 = Multiply(A=ComponentMask(None=CollectionParameter() {collection=&FortniteMaterialParameters, parameterName=CloudsShadow}) {b=true, g=true, r=true}) {constB=0.5}
  n74 = LinearInterpolate(A=@n73, B=Multiply(A=@n73, B=@n73), Alpha=NamedRerouteUsage())
  NamedRerouteDeclaration(None=VertexInterpolator(VS=QualitySwitch(Default=@n73, Low=@n73, High=@n73, Medium=@n73, Epic=@n74)).PS) {name=Shadow Color, nodeColor=<rgba 1 0 0.7031 1>}
  n75 = Multiply(A=@n47.G, B=Sign(None=@n42))
  NamedRerouteDeclaration(None=VertexInterpolator(VS=Saturate(None=@n75)).PS) {name=Mask 2, nodeColor=<rgba 1 0 0.2706 1>}
  n76 = WorldPosition() {worldPositionShaderOffset=WPT_CameraRelative}
  n77 = TransformPosition(None=Constant3Vector()) {transformSourceType=TRANSFORMPOSSOURCE_Instance, transformType=TRANSFORMPOSSOURCE_TranslatedWorld}
  n78 = Normalize(VectorInput=@n77)
  n79 = Normalize(VectorInput=CrossProduct(A=Constant3Vector() {constant=<rgba 0 0 1 0>}, B=@n78))
  n80 = LocalPosition()
  n81 = Multiply(A=@n79, B=Multiply(A=ComponentMask(None=@n80.XYZ) {r=true}) {constB=-1})
  n82 = Multiply(
   A=CrossProduct(A=@n79, B=NamedRerouteDeclaration(None=@n78) {name=Actor Vector, nodeColor=<rgba 1 0 0.9375 1>})
   B=Multiply(A=ComponentMask(None=@n80.XYZ) {b=true}) {constB=-1}
  )
  n83 = Add(
   A=Multiply(A=Subtract(A=@n76.XYZ, B=@n77)) {constB=-1}
   B=Multiply(
    A=Add(A=@n81, B=@n82)
    B=MaterialFunctionCall() {materialFunction=&ObjectScale}.Scale XYZ
   )
  ) {constB=0}
  NamedRerouteDeclaration(None=@n83) {name=WPO, nodeColor=<rgba 0.4531 1 0 1>}
  n84 = Saturate(None=Multiply(A=NamedRerouteUsage()) {constB=-1})
  n85 = Multiply(A=@n84, B=@n84)
  NamedRerouteDeclaration(None=Multiply(A=@n85, B=@n85)) {name=Halo Mask, nodeColor=<rgba 0 1 0.0859 1>}
  n86 = Multiply(A=@n75) {constB=-1}
  NamedRerouteDeclaration(None=VertexInterpolator(VS=Saturate(None=@n86)).PS) {name=Mask 3, nodeColor=<rgba 1 0 0.2706 1>}
  n87 = Multiply(A=@n47.B, B=Sign(None=@n43))
  n88 = Multiply(A=@n87) {constB=-1}
  NamedRerouteDeclaration(None=VertexInterpolator(VS=Saturate(None=@n88)).PS) {name=Mask 0, nodeColor=<rgba 1 0 0.2706 1>}
  NamedRerouteDeclaration(None=VertexInterpolator(VS=Saturate(None=@n87)).PS) {name=Mask 5, nodeColor=<rgba 1 0 0.2706 1>}
  n89 = NamedRerouteUsage()
  n90 = CrossProduct(A=LinearInterpolate(A=@n33, B=Multiply(A=@n89) {constB=-1}) {constAlpha=2}, B=@n35)
  n91 = Saturate(None=Multiply(A=DotProduct(A=@n32, B=@n90)) {constB=-1})
  n92 = Multiply(A=@n91, B=@n91)
  n93 = Multiply(A=Multiply(A=@n92, B=@n92)) {constB=4}
  NamedRerouteDeclaration(None=VertexInterpolator(VS=Saturate(None=@n93)).PS) {name=Mask 6, nodeColor=<rgba 1 0 0.2706 1>}
  NamedRerouteDeclaration(None=@n4.A) {name=Alpha, nodeColor=<rgba 0 1 0.8125 1>}
  n94 = Add(A=@n83, B=WorldPosition() {worldPositionShaderOffset=WPT_CameraRelative}.XYZ)
  NamedRerouteDeclaration(None=Normalize(VectorInput=Multiply(A=@n94) {constB=-1})) {name=Vertex CameraVector, nodeColor=<rgba 0.0312 0 1 1>}
  OneMinus(None=@n4.G)
  Subtract(A=Multiply(A=@n70) {constB=2}) {constB=0.5}
 ]}
