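# Translucent, two-sided unlit material for baked clouds in the time-of-day cloud and storm system, with fog computed per pixel.
# It uses material attributes, with the cloud shading delegated to the MF_BakedClouds material function.
# Cloud fogging is applied, and translucency renders before depth of field.
# Automatic usage flagging in the editor is turned off.
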Material /Game/TimeOfDay/CloudsAndStorm/Materials/M_BakedClouds_PixelFog
graph {
 properties {
  bApplyCloudFogging=true
  bAutomaticallySetUsageInEditor=false
  bComputeFogPerPixel=true
  bUseMaterialAttributes=true
  blendMode=BLEND_Translucent
  shadingModel=MSM_Unlit
  translucencyPass=MTP_BeforeDOF
  twoSided=true
  usedShadingModels=Unlit
 }
 nodes [
  MaterialFunctionCall() {materialFunction=&/MF_BakedClouds}
  Reroute()
 ]
}
