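# Volumetric cloud material function that builds the storm ring around the safe zone, outputting extinction, emission, color mask and linear height.
# MPC_SafeZoneClouds drives ring radius, thickness, altitude and height, blended between start and end values.
# Radial distance, rippled by animated sine waves, and normalized altitude look up a storm profile texture.
# Density combines that profile with Voronoi noise and tangentially rotating 3D flowmaps; a scrolling volume texture supplies lightning.
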MaterialFunction /Game/TimeOfDay/CloudsAndStorm/Materials/MaterialFunctions/MF_VolumetricStorm_Ch6
graph {nodes [
  CollectionParameter() {collection=/Game/Athena/SafeZone/MPC_SafeZoneClouds, parameterName=CloudStartingBias}
  CollectionParameter() {collection=&MPC_SafeZoneClouds, parameterName=CloudStartingBias}
  CollectionParameter() {collection=&MPC_SafeZoneClouds, parameterName=SafeZoneScale}
  CollectionParameter() {collection=&MPC_SafeZoneClouds, parameterName=SafeZoneScale}
  CollectionParameter() {collection=&MPC_SafeZoneClouds, parameterName=SafeZoneScale}
  CollectionParameter() {collection=&MPC_SafeZoneClouds, parameterName=CloudStartingBias}
  CollectionParameter() {collection=&MPC_SafeZoneClouds, parameterName=CloudStartingBias}
  FunctionOutput(None=NamedRerouteUsage()) {outputName=Extinction, sortPriority=3}
  FunctionOutput(None=NamedRerouteUsage()) {outputName=Emission, sortPriority=2}
  FunctionOutput(None=NamedRerouteUsage()) {outputName=Color Mask}
  FunctionOutput(None=NamedRerouteUsage()) {outputName=Linear Height, sortPriority=1}
  n1 = CloudSampleAttribute()
  MaterialFunctionCall(
   Input=@n1.NormAltitudeInLayer
   Input Low=ScalarParameter() {defaultValue=-0.02, parameterName=LowRemap}
   Input High=ScalarParameter() {defaultValue=1, parameterName=HighRemap}
  ) {materialFunction=&RemapValue_0-1}
  n2 = Time()
  n3 = CollectionParameter() {collection=&MPC_SafeZoneClouds, parameterName=CustomCloudsTime}
  n4 = Ceil(None=Saturate(None=@n3))
  Multiply(
   A=NamedRerouteDeclaration(None=LinearInterpolate(A=@n2, B=@n3, Alpha=@n4)) {name=CustomTime, nodeColor=<rgba 0 1 0.1094 1>}
   B=CollectionParameter() {collection=&MPC_SafeZoneClouds, parameterName=TopCloudsSpeed}
  )
  n5 = VectorParameter() {group=Storm, parameterName=AlbedoStormBottom, defaultValue=<rgba 0.2602 0.1591 0.2656 1>}
  n6 = VectorParameter() {group=Storm, parameterName=AlbedoStormTop, defaultValue=<rgba 0 0.0938 0.0649 1>}
  NamedRerouteDeclaration(None=AppendVector(A=LinearInterpolate(A=@n5.RGB, B=@n6.RGB), B=ComponentMask() {r=true})) {name=Color, nodeColor=<rgba 0.3594 0 1 1>}
  n7 = Distance(
   A=ComponentMask(None=WorldPosition().XYZ) {g=true, r=true}
   B=ComponentMask(None=CollectionParameter() {collection=&MPC_SafeZoneClouds, parameterName=SafeZoneLoc}) {g=true, r=true}
  )
  n8 = MaterialFunctionCall(Float2=NamedRerouteUsage()) {materialFunction=&BreakOutFloat2Components}
  n9 = MaterialFunctionCall!(Km=@n8.G) {materialFunction=/Engine/Functions/Engine_MaterialFunctions01/Units/Km-to-Cm}
  n10 = ScalarParameter() {defaultValue=1, group=StormRadialWaves, parameterName=WavesInnerRadiusMultiplier}
  n11 = Subtract(A=@n7, B=@n9.Cm)
  n12 = MaterialFunctionCall(Km=@n8.R) {materialFunction=&Km-to-Cm}
  n13 = Add(A=@n12.Cm, B=@n9.Cm)
  n14 = CloudSampleAttribute()
  n15 = Constant()
  n16 = ScalarParameter() {defaultValue=1, parameterName=HighRemap}
  n17 = NamedRerouteUsage()
  n18 = Divide(
   A=Subtract(
    A=MaterialFunctionCall(Input=@n14.NormAltitudeInLayer, Input Low=@n15, Input High=@n16) {materialFunction=&RemapValue_0-1}.Result
    B=ComponentMask(None=@n17) {b=true}
   )
   B=ComponentMask(None=@n17) {a=true}
  )
  n19 = Add(A=Saturate(None=Divide(A=@n11, B=@n13)), B=Saturate(None=OneMinus(None=@n18)))
  n20 = Add(
   A=MaterialFunctionCall!(Vector or UVs=NamedRerouteUsage()) {materialFunction=/Game/Athena/SafeZone/S11/MF_VectorToRadialValueFast}.Result
   B=Multiply(A=Time(), B=ScalarParameter() {group=StormRadialWaves, parameterName=WavesSpeed})
  )
  n21 = Sine(None=Multiply(A=Frac(None=@n20), B=ScalarParameter() {defaultValue=4, group=StormRadialWaves, parameterName=WaveCount}) {constB=10})
  n22 = Power(
   Base=Add(A=Multiply(A=@n21) {constB=0.5}) {constB=0.5}
   Exp=ScalarParameter() {defaultValue=1, group=StormRadialWaves, parameterName=WavesPow}
  )
  n23 = ComponentMask(None=Saturate(None=Add(A=@n19, B=@n22) {constB=0})) {r=true}
  n24 = OneMinus(None=@n23)
  n25 = Multiply(A=@n9.Cm, B=Saturate(None=Add(A=@n10, B=@n24)))
  NamedRerouteDeclaration(None=Divide(A=Subtract(A=@n7, B=@n25), B=@n12.Cm)) {name=Storm Radius, nodeColor=<rgba 0.0547 1 0 1>}
  n26 = NamedRerouteUsage()
  n27 = ScalarParameter() {defaultValue=0.001, group=Storm, parameterName=scc}
  n28 = Saturate(None=Add(A=Multiply(A=@n26, B=@n27), B=ScalarParameter() {defaultValue=0.5, group=Storm, parameterName=off}))
  n29 = Divide(A=LinearInterpolate(
    A=ScalarParameter() {defaultValue=128, group=Storm, parameterName=Global Storm Density}
    B=ScalarParameter() {defaultValue=128, group=Storm, parameterName=Global Storm Density Top}
    Alpha=@n28
   )) {constB=100}
  n30 = Multiply(
   A=ComponentMask(None=NamedRerouteUsage()) {r=true}
   B=LinearInterpolate(
    A=ScalarParameter() {defaultValue=0.5286, group=StormDFs, parameterName=Noise Intensity}
    B=ScalarParameter() {defaultValue=0.2, group=StormDFs, parameterName=Noise IntensityTop}
    Alpha=@n28
   )
  )
  n31 = Multiply(A=NamedRerouteUsage(), B=VectorParameter() {group=Storm, parameterName=WorldScaleNoise, defaultValue=<rgba 1 1 1 0>}.RGB)
  n32 = LinearInterpolate(
   A=ScalarParameter() {defaultValue=1, group=Storm, parameterName=NoiseGlobalScaleEnd}
   B=ScalarParameter() {defaultValue=1, group=Storm, parameterName=NoiseGlobalScale}
   Alpha=NamedRerouteUsage()
  )
  n33 = Multiply(A=ScalarParameter() {defaultValue=4500000, group=Storm, parameterName=scale}, B=@n32)
  n34 = TextureObjectParameter() {parameterName=VolumeTextureVoronoiSimple, texture=/Game/TimeOfDay/CloudsAndStorm/Textures/T_EmptyTexture_Volume}
  n35 = TextureSample!(
   UVs=Divide(A=@n31, B=MaterialFunctionCall(Km=@n33) {materialFunction=&Km-to-Cm}.Cm)
   Tex=@n34
   Level=CollectionParameter() {collection=&MPC_SafeZoneClouds, parameterName=StormDetailVolumeMip}
   Apply View MipBias
  ) {mipValueMode=TMVM_MipLevel, samplerType=SAMPLERTYPE_LinearColor, texture=/Volumetrics/Content/VolumeTextures/Textures/VT_Voronoi_Simple}
  n36 = ScalarParameter() {defaultValue=0.8, group=StormDFs, parameterName=Noise DF Center End}
  n37 = ScalarParameter() {defaultValue=0.8, group=StormDFs, parameterName="Noise DF Center "}
  n38 = NamedRerouteUsage()
  n39 = NamedRerouteDeclaration(None=Subtract(A=@n35.R, B=LinearInterpolate(A=@n36, B=@n37, Alpha=@n38)) {constB=0.5}) {name=StormNoise, nodeColor=<rgba 1 0 0.8438 1>}
  n40 = MaterialFunctionCall(Float2=Normalize(VectorInput=NamedRerouteUsage())) {materialFunction=&BreakOutFloat2Components}
  n41 = AppendVector(A=Multiply(A=@n40.G) {constB=-1}, B=@n40.R)
  n42 = Constant()
  n43 = Multiply(A=@n32, B=ScalarParameter() {defaultValue=512, group=Storm, parameterName=scale2})
  n44 = MaterialFunctionCall!(
   Volume Texture=@n34
   UVW=Divide(A=@n31, B=MaterialFunctionCall(Km=@n43) {materialFunction=&Km-to-Cm}.Cm)
   Velocity=Multiply(A=AppendVector(A=@n41, B=@n42), B=ScalarParameter() {defaultValue=0.7, group=StormFlowMap, parameterName=Storm Rotation Velocity})
   Override Time=Multiply(A=ScalarParameter() {defaultValue=0.1, group=StormFlowMap, parameterName=T}, B=NamedRerouteUsage())
   Mip Level=CollectionParameter() {collection=&MPC_SafeZoneClouds, parameterName=StormFlowMapsVolumeMip}
  ) {materialFunction=&/Flowmaps_3D}
  n45 = ComponentMask(None=@n44.Result RGB) {r=true}
  n46 = LinearInterpolate(
   A=ScalarParameter() {defaultValue=0.3, group=StormDFs, parameterName=Flowmaps DF Center End}
   B=ScalarParameter() {defaultValue=0.8, group=StormDFs, parameterName="Flowmaps DF Center "}
   Alpha=CollectionParameter() {collection=&MPC_SafeZoneClouds, parameterName=SafeZoneScale}
  )
  n47 = ScalarParameter() {defaultValue=1, group=StormDFs, parameterName="Flowmaps DF Center PreStart "}
  n48 = Saturate(None=NamedRerouteDeclaration(None=NamedRerouteUsage()) {name=CloudStartingBias, nodeColor=<rgba 0 0.3672 1 1>})
  n49 = Multiply(
   A=Subtract(A=NamedRerouteDeclaration(None=@n45) {name=FlowMap, nodeColor=<rgba 1 0 0.1172 1>}, B=LinearInterpolate(A=@n46, B=@n47, Alpha=@n48)) {constB=0.5}
   B=ScalarParameter() {defaultValue=0.5, group=StormDFs, parameterName=Flowmaps Intensity}
  )
  MaterialExpressionNamedRerouteUsage_4 = NamedRerouteUsage()
  MaterialExpressionComponentMask_15 = ComponentMask(None=NamedRerouteUsage()) {g=true}
  n50 = Custom!(NormAltitude=@MaterialExpressionNamedRerouteUsage_4, Density=@MaterialExpressionComponentMask_15) {
   code="float HeightDensity = Density;\r\n\r\nif(NormAltitude<0.05 || NormAltitude>0.95)\r\n{\r\n   HeightDensity = 0.0;\r\n}\r\n\r\nreturn HeightDensity;"
   inputs [
    {
     inputName=NormAltitude
     input {expression=&self:MaterialExpressionNamedRerouteUsage_4}
    }
    {
     inputName=Density
     input {expression=&self:MaterialExpressionComponentMask_15}
    }
   ]
  }
  n51 = ScalarParameter() {defaultValue=0.681, group=StormDFs, parameterName=StormDonut DF Center}
  n52 = Add(
   A=Add(A=Multiply(A=@n30, B=@n39), B=NamedRerouteDeclaration(None=@n49) {name=Flow, nodeColor=<rgba 0 1 0.5547 1>})
   B=Multiply(A=ScalarParameter() {defaultValue=0.5286, group=StormDFs, parameterName=StormDonut Intensity}, B=Subtract(A=@n50, B=@n51))
  )
  n53 = Max(A=NamedRerouteUsage(), B=CollectionParameter() {collection=&MPC_SafeZoneClouds, parameterName=PreventStormShrinking})
  n54 = Multiply(
   A=CollectionParameter() {collection=&MPC_SafeZoneClouds, parameterName=EnableVolumetricStorm}
   B=Saturate(None=OneMinus(None=@n53))
  )
  NamedRerouteDeclaration(None=Multiply(A=Multiply(A=@n29, B=@n52), B=@n54)) {name=StormEXT, nodeColor=<rgba 1 0 0.8894 1>}
  n55 = CollectionParameter() {collection=&MPC_SafeZoneClouds, parameterName=SafeZoneLoc}
  NamedRerouteDeclaration(None=Subtract(A=WorldPosition().XYZ, B=ComponentMask(None=@n55) {b=true, g=true, r=true})) {name=Safe Zone Local, nodeColor=<rgba 1 0 0.7266 1>}
  n56 = MaterialFunctionCall() {materialFunction=&/MF_RemappedStormScale}
  NamedRerouteDeclaration(None=@n56.StartingBias) {name=StartingBiasScaleBased, nodeColor=<rgba 1 0 0.0546 1>}
  n57 = WorldPosition()
  n58 = MaterialFunctionCall(Km=ScalarParameter() {defaultValue=5, group=Lightning, parameterName=Lighting Pattern Scale}) {materialFunction=&Km-to-Cm}
  n59 = AppendVector(
   A=ComponentMask(None=Divide(A=@n57.XYZ, B=@n58.Cm)) {g=true, r=true}
   B=MaterialFunctionCall(Speed=ScalarParameter() {defaultValue=0.2, group=Lightning, parameterName=Lighting Speed}) {materialFunction=&TimeWithSpeedVariable}.Time
  )
  n60 = TextureObjectParameter() {parameterName=VolumeTextureLightning, texture=&T_EmptyTexture_Volume}
  n61 = TextureSample!(UVs=@n59, Tex=@n60, Apply View MipBias) {texture=/Volumetrics/Content/VolumeTextures/Textures/VT_Lightning}
  n62 = ScalarParameter() {defaultValue=3, group=Lightning, parameterName=Lightning Power}
  n63 = Multiply(A=@n61.R, B=@n61.R)
  n64 = ScalarParameter() {group=Lightning, parameterName=Fill Lightning Scatter}
  n65 = ScalarParameter() {defaultValue=2.5, group=Lightning, parameterName=Lightning Mip}
  n66 = Add(
   A=Add(A=Power(Base=@n61.R, Exp=@n62), B=Multiply(A=@n63, B=@n64))
   B=Multiply(
    A=TextureSample!(UVs=@n59, Tex=@n60, Level=@n65, Apply View MipBias) {mipValueMode=TMVM_MipLevel, texture=&VT_Lightning}.R
    B=ScalarParameter() {defaultValue=0.05, group=Lightning, parameterName=Lightning Spread}
   )
  ) {constB=0}
  n67 = NamedRerouteUsage()
  n68 = Saturate(None=Multiply(A=@n67) {constB=50})
  NamedRerouteDeclaration(None=Multiply(A=Multiply(A=@n66, B=@n68), B=VectorParameter() {group=Lightning, parameterName=Lightning Color, defaultValue=<rgba 0.0997 0.3491 2 0>}.RGB)) {name=Lightning, nodeColor=<rgba 1 0 0.2223 1>}
  NamedRerouteDeclaration(None=@n54) {name=Storm Control, nodeColor=<rgba 1 0 0.7734 1>}
  n69 = NamedRerouteUsage()
  n70 = NamedRerouteUsage()
  n71 = TextureSample(
   UVs=MaterialFunctionCall(X=Saturate(None=@n69), Y=OneMinus(None=@n70)) {materialFunction=&MakeFloat2}.Result
   Level=CollectionParameter() {collection=&MPC_SafeZoneClouds, parameterName=StormProfileMip}
  ) {mipValueMode=TMVM_MipLevel, samplerType=SAMPLERTYPE_LinearColor, texture=/Game/TimeOfDay/CloudsAndStorm/Textures/StormProfileLegacy}
  NamedRerouteDeclaration(None=@n71.A) {name=StormShape, nodeColor=<rgba 1 0.141 0.7007 1>}
  n72 = OneMinus(None=NamedRerouteUsage())
  n73 = NamedRerouteUsage()
  n74 = Multiply(
   A=Add(A=Add(A=@n72, B=@n73), B=ScalarParameter() {defaultValue=-0.7, parameterName=StormCloudMaskStartMultiplier})
   B=ScalarParameter() {defaultValue=1, parameterName=StormCloudMaskColorStrength}
  )
  n75 = Add(A=Multiply(A=NamedRerouteUsage()) {constB=2}) {constB=-0.5}
  NamedRerouteDeclaration(None=LinearInterpolate(A=Saturate(None=@n74), Alpha=Saturate(None=@n75)) {constA=1}) {name=StormCloudColorMask, nodeColor=<rgba 1 0 0.0703 1>}
  n76 = NamedRerouteUsage()
  n77 = Subtract(A=@n1.NormAltitudeInLayer, B=ComponentMask(None=@n76) {b=true})
  n78 = Add(
   A=ComponentMask(None=@n76) {a=true}
   B=Multiply(A=ScalarParameter() {defaultValue=-0.0857, group=StormRadialWaves, parameterName=WavesHeight}, B=@n23)
  )
  NamedRerouteDeclaration(None=Saturate(None=Divide(A=@n77, B=@n78))) {name=Linear Height Profile, nodeColor=<rgba 1 0.1875 0 1>}
  NamedRerouteDeclaration(None=@n56.ZoneScale) {name=ZoneScaleRemapped, nodeColor=<rgba 0.0078 0 1 1>}
  NamedRerouteDeclaration(None=@n71.RGB) {name=Mapped Storm Profile, nodeColor=<rgba 0.8984 0 1 1>}
  n79 = CollectionParameter() {collection=&MPC_SafeZoneClouds, parameterName=StormThicknessKm-End}
  n80 = CollectionParameter() {collection=&MPC_SafeZoneClouds, parameterName=StormRadiusKm-End}
  n81 = CollectionParameter() {collection=&MPC_SafeZoneClouds, parameterName=StormAltitude-End}
  n82 = CollectionParameter() {collection=&MPC_SafeZoneClouds, parameterName=StormHeight-End}
  n83 = CollectionParameter() {collection=&MPC_SafeZoneClouds, parameterName=StormThicknessKm}
  n84 = Saturate(None=NamedRerouteUsage())
  n85 = Add(
   A=CollectionParameter() {collection=&MPC_SafeZoneClouds, parameterName=StormRadiusKm}
   B=Multiply(A=ScalarParameter() {defaultValue=3, parameterName=StormRadius_PreStartOffset}, B=Multiply(A=@n84, B=@n84))
  )
  n86 = CollectionParameter() {collection=&MPC_SafeZoneClouds, parameterName=StormAltitude}
  n87 = CollectionParameter() {collection=&MPC_SafeZoneClouds, parameterName=StormHeight}
  NamedRerouteDeclaration(None=LinearInterpolate(
    A=MaterialFunctionCall(X=@n79, Y=@n80, Z=@n81, A=@n82) {materialFunction=&MakeFloat4}.Result
    B=MaterialFunctionCall(X=@n83, Y=@n85, Z=@n86, A=@n87) {materialFunction=&MakeFloat4}.Result
    Alpha=NamedRerouteUsage()
   )) {name=Storm Collection Params, nodeColor=<rgba 0.4708 0 1 1>}
  NamedRerouteDeclaration(None=@n28) {name=StormHeightMask, nodeColor=<rgba 1 0 0.2578 1>}
  NamedRerouteDeclaration(None=@n44.UVW_A) {name=FlowUVs, nodeColor=<rgba 0 0.5312 1 1>}
  Power(Base=@n67, Exp=ScalarParameter() {defaultValue=0.1, group=Lightning, parameterName=Lightning Mask Power})
 ]}
