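# Additive volume material for the Hermes thin volumetric clouds, blending high cloud layers with a Chapter 6 storm.
# Cloud layers sample channels of one high-cloud texture, some panning over time, eroded by altitude masks and Voronoi volume noise.
# MF_VolumetricStorm_Ch6 supplies storm extinction, emission, height and color mask for tinting albedo and emissive.
# Max-combined, masked density feeds conservative density and scales emissive, which is zeroed in reflection captures.
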Material /Game/TimeOfDay/CloudsAndStorm/Materials/M_HermesVolumetricClouds_Thin
graph {
 properties {
  bAutomaticallySetUsageInEditor=false
  bUsedWithVolumetricCloud=true
  blendMode=BLEND_Additive
  materialDomain=MD_Volume
  previewMesh=/Game/Environments/Hera/FX/Meshes/SM_WaterTankBreak_GlassCracks_Water_01
  usedShadingModels=Default Lit
 }
 nodes [
  n1 = Divide(
   A=ComponentMask(None=NamedRerouteUsage()) {g=true, r=true}
   B=ComponentMask(None=VectorParameter() {parameterName=Layer3_WorldTileDivisor, defaultValue=<rgba 750000 750000 750000 0>}.RGB) {g=true, r=true}
  )
  n2 = Power(
   Base=NamedRerouteDeclaration(None=TextureSample(UVs=@n1) {texture=/Game/TimeOfDay/CloudsAndStorm/Textures/T_Ch6_HighClouds}.B) {name=Layer3, nodeColor=<rgba 0.2188 1 0 1>}
   Exp=ScalarParameter() {defaultValue=1.6, parameterName=Layer3_ExtictionExponent}
  )
  n3 = CloudSampleAttribute()
  n4 = ScalarParameter() {defaultValue=0.98, parameterName=Layer3_HeightCenter}
  n5 = ScalarParameter() {defaultValue=0.07, parameterName=Layer3_HeightRadius}
  n6 = ScalarParameter() {defaultValue=0.2, parameterName=Layer3_HeightHardness}
  n7 = Power(
   Base=OneMinus(None=SphereMask(A=@n3.NormAltitudeInLayer, B=@n4, Radius=@n5, Hardness=@n6))
   Exp=ScalarParameter() {defaultValue=1, parameterName=Layer3_HeightErosionExponent}
  )
  n8 = Multiply(A=Saturate(None=Subtract(A=@n2, B=@n7)), B=ScalarParameter() {defaultValue=0.004, parameterName=Layer3_Extinction})
  n9 = NamedRerouteUsage()
  n10 = ScalarParameter() {defaultValue=0.5, parameterName=Layer3_DetailNoiseScale}
  Add(A=@n8, B=Multiply(A=@n8, B=Multiply(A=@n9, B=@n10)))
  n11 = Time()
  Add(
   A=Multiply(A=@n11, B=ScalarParameter() {parameterName=GlobalTimeFactor})
   B=ScalarParameter() {defaultValue=200, parameterName=GlobalTimeOffset}
  )
  Ceil()
  Floor()
  n12 = VectorParameter() {parameterName=Storm Color Botom, defaultValue=<rgba 0.1027 0.2048 0.5417 1>}
  n13 = VectorParameter() {parameterName=Storm Color Top, defaultValue=<rgba 0.6737 0.2919 0.9427 1>}
  n14 = NamedRerouteUsage()
  n15 = VectorParameter() {parameterName=AlbedoColor, defaultValue=<rgba 1 0.6707 0.3372 1>}
  LinearInterpolate(
   A=Multiply(A=LinearInterpolate(A=@n12.RGB, B=@n13.RGB, Alpha=@n14), B=@n15.RGB)
   B=Multiply(A=@n15.RGB, B=VectorParameter() {parameterName=Color Tint, defaultValue=<rgba 1 1 1 1>}.RGB)
   Alpha=NamedRerouteUsage()
  )
  n16 = VectorParameter() {parameterName=EmissiveColor, defaultValue=<rgba 0.0718 0.0654 0.1094 1>}
  n17 = Multiply(
   A=LinearInterpolate(
    A=VectorParameter() {parameterName=Storm Emission Bottom, defaultValue=<rgba 0.0336 0.0278 0.1979 1>}.RGB
    B=VectorParameter() {parameterName=Storm Emission Top, defaultValue=<rgba 0.1414 0.0639 0.2604 1>}.RGB
    Alpha=NamedRerouteUsage()
   )
   B=@n16.RGB
  )
  n18 = Multiply(A=@n16.RGB, B=VectorParameter() {parameterName=Color Tint, defaultValue=<rgba 1 1 1 1>}.RGB)
  n19 = NamedRerouteUsage()
  n20 = Max(A=NamedRerouteUsage(), B=NamedRerouteUsage())
  n21 = Max(A=Multiply(A=Saturate(None=@n20), B=NamedRerouteUsage()), B=NamedRerouteUsage())
  Multiply(
   A=Add(A=LinearInterpolate(A=@n17, B=@n18, Alpha=@n19), B=NamedRerouteUsage())
   B=ReflectionCapturePassSwitch!(Default=@n21, Reflection=Constant())
  )
  n22 = Subtract(
   A=WorldPosition() {worldPositionShaderOffset=WPT_CameraRelative}.XYZ
   B=Multiply(A=ActorPositionWS() {originType=CameraRelative}, B=Constant3Vector() {constant=<rgba 3 3 1 0>}) {constB=3}
  )
  n23 = Divide(
   A=ComponentMask(None=NamedRerouteDeclaration(None=@n22) {name=World Position Parallax, nodeColor=<rgba 0 1 0.4844 1>}) {g=true, r=true}
   B=ComponentMask(None=VectorParameter() {parameterName=Layer1_WorldTileDivisor, defaultValue=<rgba 1000000 1000000 1000000 0>}.RGB) {g=true, r=true}
  )
  n24 = Multiply(
   A=NamedRerouteUsage()
   B=AppendVector(
    A=ScalarParameter() {defaultValue=-0.002, parameterName=Layer1_PanSpeed_X}
    B=ScalarParameter() {defaultValue=-0.004, parameterName=Layer1_PanSpeed_Y}
   )
  )
  n25 = Subtract(
   A=TextureSample(UVs=Add(A=@n23, B=@n24)) {texture=&T_Ch6_HighClouds}.R
   B=Subtract(B=ScalarParameter() {parameterName=Layer1_Coverage}) {constA=0.2}
  )
  n26 = Add(A=Multiply(A=NamedRerouteUsage(), B=ScalarParameter() {defaultValue=3.2016, parameterName=Layer1_DetailNoiseScale}))
  n27 = Constant() {r=1}
  n28 = Divide(
   A=Subtract(A=CloudSampleAttribute().NormAltitudeInLayer, B=ScalarParameter() {defaultValue=0.5, parameterName=Layer1_HeightCenter})
   B=ScalarParameter() {defaultValue=0.35, parameterName=Layer1_HeightRadius}
  )
  n29 = Saturate(None=Multiply(A=@n28) {constB=-1})
  n30 = Saturate(None=Abs(None=@n28))
  n31 = ScalarParameter() {defaultValue=1.6422, parameterName=Layer1_HeightHardness}
  n32 = Subtract(
   A=Power(Base=Saturate(None=@n25), Exp=LinearInterpolate(A=@n26, B=@n27, Alpha=@n29) {constA=1})
   B=NamedRerouteDeclaration(None=Power(Base=@n30, Exp=@n31)) {name=Layer1HeightMask, nodeColor=<rgba 1 0.1172 0 1>}
  )
  NamedRerouteDeclaration(None=Multiply(A=Saturate(None=@n32), B=ScalarParameter() {defaultValue=0.074, parameterName=Layer1_Extinction})) {name=Layer1, nodeColor=<rgba 0.2187 1 0 1>}
  n33 = Subtract(A=WorldPosition() {worldPositionShaderOffset=WPT_CameraRelative}.XYZ, B=ActorPositionWS() {originType=CameraRelative})
  n34 = VectorParameter() {parameterName=Layer2_WorldTileDivisor, defaultValue=<rgba 832000 832000 832000 0>}
  n35 = ScalarParameter() {defaultValue=-0.0013, parameterName=Layer2_PanSpeed_X}
  n36 = ScalarParameter() {defaultValue=-0.0025, parameterName=Layer2_PanSpeed_Y}
  n37 = Add(
   A=Divide(A=ComponentMask(None=@n33) {g=true, r=true}, B=ComponentMask(None=@n34.RGB) {g=true, r=true})
   B=Multiply(A=NamedRerouteUsage(), B=AppendVector(A=@n35, B=@n36))
  )
  n38 = SphereMask(
   A=CloudSampleAttribute().NormAltitudeInLayer
   B=ScalarParameter() {defaultValue=0.9, parameterName=Layer2_HeightCenter}
   Radius=ScalarParameter() {defaultValue=0.07, parameterName=Layer2_HeightRadius}
   Hardness=ScalarParameter() {defaultValue=0.2, parameterName=Layer2_HeightHardness}
  )
  n39 = Subtract(
   A=Power(
    Base=TextureSample(UVs=@n37) {texture=&T_Ch6_HighClouds}.B
    Exp=ScalarParameter() {defaultValue=1, parameterName=Layer2_ExtictionExponent}
   )
   B=Power(Base=OneMinus(None=@n38), Exp=ScalarParameter() {defaultValue=1, parameterName=Layer2_HeightErosionExponent})
  )
  NamedRerouteDeclaration(None=Multiply(A=Saturate(None=@n39), B=ScalarParameter() {defaultValue=0.005, parameterName=Layer2_Extinction})) {name=Layer2, nodeColor=<rgba 0.2188 1 0 1>}
  n40 = MaterialFunctionCall() {materialFunction=/Game/TimeOfDay/CloudsAndStorm/Materials/MaterialFunctions/MF_VolumetricStorm_Ch6}
  NamedRerouteDeclaration(None=@n40.Emission) {name=Storm Emission, nodeColor=<rgba 0.4766 1 0 1>}
  NamedRerouteDeclaration(None=@n40.Extinction) {name=Storm Extinction, nodeColor=<rgba 1 0 0.7969 1>}
  NamedRerouteDeclaration(None=@n40.Linear Height) {name=Linear Height, nodeColor=<rgba 1 0.2359 0.1161 1>}
  n41 = CameraPositionWS()
  n42 = WorldPosition() {worldPositionShaderOffset=WPT_CameraRelative}
  n43 = ScalarParameter() {defaultValue=200000, parameterName=NearCameraFadeDistance}
  n44 = ScalarParameter() {defaultValue=0.2, parameterName=NearCameraFadeHardness}
  NamedRerouteDeclaration(None=OneMinus(None=SphereMask(A=@n41, B=@n42.XYZ, Radius=@n43, Hardness=@n44))) {name=NearCameraMask, nodeColor=<rgba 0.6641 0 1 1>}
  n45 = NamedRerouteUsage()
  n46 = ScalarParameter() {defaultValue=120977, parameterName=DetailNoiseDivisor}
  n47 = NamedRerouteUsage()
  n48 = VectorParameter() {parameterName=DetailNoisePanSpeed, defaultValue=<rgba 0.01 -0.0111 -0.0222 0>}
  n49 = TextureSample(
   UVs=Add(A=Divide(A=@n45, B=@n46), B=Multiply(A=@n47, B=@n48.RGB))
   Tex=TextureObjectParameter() {parameterName=VolumeTextureVoronoi, texture=/Game/TimeOfDay/CloudsAndStorm/Textures/T_EmptyTexture_Volume}
  ) {samplerType=SAMPLERTYPE_LinearColor, texture=/Game/TimeOfDay/CloudsAndStorm/Textures/FN_VT_Voronoi}
  NamedRerouteDeclaration(None=@n49.R) {name=VolumeNoise, nodeColor=<rgba 1 0 0.75 1>}
  NamedRerouteDeclaration(None=SphereMask(
    A=Constant()
    B=WorldPosition() {worldPositionShaderOffset=WPT_CameraRelative}.XY
    Radius=ScalarParameter() {defaultValue=732146.75, parameterName=IslandMaskRadius}
    Hardness=ScalarParameter() {defaultValue=0.3, parameterName=IslandMaskHardness}
   )) {name=IslandMask, nodeColor=<rgba 0 1 0.3438 1>}
  NamedRerouteDeclaration(None=@n11) {name=Time, nodeColor=<rgba 0.4062 0 1 1>}
  NamedRerouteDeclaration(None=@n40.Color Mask) {name=Color Mask, nodeColor=<rgba 0 0.1562 1 1>}
  ScalarParameter() {defaultValue=0.5, parameterName=AO}
  Subtract(A=@n49.R) {constB=0.5}
  Time()
  Time()
  VolumetricAdvancedMaterialOutput!(
   PhaseG=ScalarParameter() {defaultValue=0.7, parameterName=PhaseG}
   PhaseG2=ScalarParameter() {defaultValue=0.01, parameterName=PhaseG2}
   PhaseBlend=ScalarParameter() {defaultValue=0.5, parameterName=PhaseBlend}
   MultiScatteringContribution=ScalarParameter() {defaultValue=0.5, parameterName=MultiScatteringContribution}
   MultiScatteringOcclusion=ScalarParameter() {defaultValue=0.5, parameterName=MultiScatteringOcclusion}
   MultiScatteringEccentricity=ScalarParameter() {defaultValue=0.5, parameterName=MultiScatteringEccentricity}
   ConservativeDensity=@n21
  ) {multiScatteringApproximationOctaveCount=1}
 ]
}
