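# Material function in the time-of-day clouds and storm materials that converts kilometers to centimeters.
# It multiplies a two-component Km input by a constant 100000 and returns the result as the Cm output.
# The conversion runs in one direction only, and the function is listed under the Units library category.
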MaterialFunction /Game/TimeOfDay/CloudsAndStorm/Materials/MaterialFunctions/Km-to-Cm
graph {
 properties {description=Converts between Kilometers and Centimeters., libraryCategoriesText [Units]}
 nodes [FunctionOutput(None=Multiply(A=FunctionInput() {inputName=Km, inputType=FunctionInput_Vector2}) {constB=100000}) {outputName=Cm}]
}
