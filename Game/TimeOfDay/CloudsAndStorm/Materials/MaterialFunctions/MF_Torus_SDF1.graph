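# Material function that computes a torus signed distance field for shaping volumetric clouds in the time-of-day storm system.
# World XY, shifted by a vector noise offset and the centre input, gives distance from the major radius (ta).
# That pairs with height, defaulting to normalized cloud-layer altitude and adjusted by scaled Z noise and vertical offset; the pair's length minus tube radius (tb) is TorusSDF.
# A thresholded-mask branch and a LocalPosition call are unconnected.
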MaterialFunction /Game/TimeOfDay/CloudsAndStorm/Materials/MaterialFunctions/MF_Torus_SDF1
graph {nodes [
  n1 = WorldPosition()
  n2 = FunctionInput(Preview=Constant3Vector()) {bUsePreviewValueAsDefault=true, inputName=VectorNoiseOffset}
  n3 = Subtract(
   A=ComponentMask(None=FunctionInput(Preview=@n1.XYZ) {bUsePreviewValueAsDefault=true, inputName=worldXY_UV, inputType=FunctionInput_Vector2}) {g=true, r=true}
   B=ComponentMask(None=@n2) {g=true, r=true}
  )
  n4 = Constant() {r=210.3907}
  n5 = Constant() {r=99.5014}
  n6 = ComponentMask(None=FunctionInput(Preview=MaterialFunctionCall(X=@n4, Y=@n5) {materialFunction=&MakeFloat2}.Result) {bUsePreviewValueAsDefault=true, inputName=torusCentre, inputType=FunctionInput_Vector2}) {g=true, r=true}
  n7 = Subtract(
   A=Distance(A=Add(A=@n3, B=@n6), B=Constant())
   B=FunctionInput(Preview=Constant() {r=102.6012}) {bUsePreviewValueAsDefault=true, inputName=ta, inputType=FunctionInput_Scalar}
  )
  n8 = ComponentMask(None=@n2) {b=true}
  n9 = FunctionInput(Preview=Constant() {r=1}) {bUsePreviewValueAsDefault=true, desc=Scale on the Z component of noise Offset, inputName=ZNoiseScale, inputType=FunctionInput_Scalar}
  n10 = Add(
   A=Subtract(A=NamedRerouteUsage(), B=Multiply(A=@n8, B=@n9) {constB=0.2})
   B=FunctionInput(Preview=Constant()) {bUsePreviewValueAsDefault=true, inputName=TorusVerticalOffset, inputType=FunctionInput_Scalar}
  )
  n11 = Subtract(
   A=Distance(
    A=MaterialFunctionCall(X=@n7, Y=@n10) {materialFunction=&MakeFloat2}.Result
    B=Constant()
   )
   B=FunctionInput(Preview=Constant() {r=19.7547}) {bUsePreviewValueAsDefault=true, inputName=tb, inputType=FunctionInput_Scalar}
  )
  FunctionOutput(None=ComponentMask(None=@n11) {r=true}) {outputName=TorusSDF}
  MaterialFunctionCall() {materialFunction=&LocalPosition}
  n12 = OneMinus(None=@n11)
  Multiply(A=Saturate(None=Sign(None=@n12)), B=Constant() {r=0.008})
  NamedRerouteDeclaration(None=FunctionInput(Preview=CloudSampleAttribute().NormAltitudeInLayer) {bUsePreviewValueAsDefault=true, inputName=HeightZPos, inputType=FunctionInput_Scalar}) {name=Z, nodeColor=<rgba 0.2891 0 1 1>}
 ]}
